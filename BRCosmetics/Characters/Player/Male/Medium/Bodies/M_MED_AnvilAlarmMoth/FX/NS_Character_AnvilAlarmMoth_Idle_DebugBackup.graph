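# Debug backup of the idle Niagara effect for the AnvilAlarmMoth character: an audio-reactive heart and mesh voxels on the body.
# System-level audio values, with a fake reactivity loop module available, drive spawn rate, scale, color, jitter and material parameters.
# Limb voxels split into low, medium and high band partitions, traveling along curves built from offset skeletal-mesh samples.
# Disabled emitters include audio readouts and static limb-path previews.
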NiagaraSystem /BRCosmetics/Characters/Player/Male/Medium/Bodies/M_MED_AnvilAlarmMoth/FX/NS_Character_AnvilAlarmMoth_Idle_DebugBackup
dependencies {
 usedDataInterfaces [&NiagaraDataInterfaceCurve, &NiagaraDataInterfaceSkeletalMesh, &NiagaraDataInterfaceColorCurve]
 usedDynamicInputs [
  &Multiply_Float
  &MakeCustomFloatFromBool
  /Niagara/DynamicInputs/Lerp/SmoothLerpOverTime_Float
  &FN_FrontEndFloatModifier
  &RemapRange
  &SetBoolByFloatComparison
  &ColorFromCurve
 ]
 usedModules [
  &self:SystemSpawnScript.SystemScriptSource.SystemScriptGraph.NiagaraNodeAssignment_4.SetVariables_#1
  &self:SystemSpawnScript.SystemScriptSource.SystemScriptGraph.NiagaraNodeAssignment_5.SetVariables_#2
  &SystemState
  &self:SystemSpawnScript.SystemScriptSource.SystemScriptGraph.NiagaraNodeAssignment_3.SetVariables_#3
  /Game/Effects/Niagara/Modules/Audio/FN_FakeAudioReactivityLoop
  &self:SystemSpawnScript.SystemScriptSource.SystemScriptGraph.NiagaraNodeAssignment_1.SetVariables_#4
  &self:SystemSpawnScript.SystemScriptSource.SystemScriptGraph.NiagaraNodeAssignment_0.SetVariables_#5
  &self:SystemSpawnScript.SystemScriptSource.SystemScriptGraph.NiagaraNodeAssignment_10.SetVariables_#6
 ]
}
emitters {
 FN_Debug001 {
  bEnabled=false
  emitterUpdateScript [
   &EmitterState(Life Cycle Mode=::NewEnumerator1, Loop Behavior=::NewEnumerator1, Loop Duration Mode=::NewEnumerator1, UseLoopDelay=false)
   &SpawnBurst_Instantaneous(Spawn Count=4, Use Spawn Probability=false)
  ]
  particleSpawnScript [
   &PartitionParticles(Number Of Partitions=4)
   &InitializeParticle(UsePositionOffset=false, Write Material Random=true)
  ]
  particleUpdateScript [
   &ParticleState(Kill Particles When Lifetime Has Elapsed=false, Loop Particles Lifetime=true)
   &DynamicMaterialParameters(
    Index 0 Param 1=&MergeHelper_FloatSwitch(
     0=$System.AudioLow
     1=$System.AudioMedium
     2=$System.AudioHigh
     3=$NPC.NPC_Audio_Reactive_Parameters.Amplitude_Music
     Index=$Particles.PartitionParticles.Partition
     Num=::NewEnumerator2
    )
    Module.Param0WriteEnabled=true
    Module.Param1WriteEnabled=true
    Module.Param2WriteEnabled=true
    Module.Param3WriteEnabled=true
   )
   /SetVariables(Particles.Position=&MergeHelper_VectorSwitch(0=<100 0 200>, 1=<75 0 200>, 2=<50 0 200>, 3=<25 0 200>, Index=$Particles.PartitionParticles.Partition, Num=::NewEnumerator2)) {isSetParameters=true, module=SetVariables_#7}
  ]
  rendererClasses [&NiagaraSpriteRendererProperties]
  renderers [{rendererClass=&NiagaraSpriteRendererProperties}]
 }
 FN_Debug002 {
  bEnabled=false
  emitterUpdateScript [
   &EmitterState(Life Cycle Mode=::NewEnumerator1, Loop Behavior=::NewEnumerator1, Loop Duration Mode=::NewEnumerator1, UseLoopDelay=false)
   &SpawnBurst_Instantaneous(Spawn Count=4, Use Spawn Probability=false)
  ]
  particleSpawnScript [
   &PartitionParticles(Number Of Partitions=4)
   &InitializeParticle(UsePositionOffset=false, Write Material Random=true)
  ]
  particleUpdateScript [
   &ParticleState(Kill Particles When Lifetime Has Elapsed=false, Loop Particles Lifetime=true)
   &DynamicMaterialParameters(
    Index 0 Param 1=&MergeHelper_FloatSwitch(
     0=$NPC.NPC_Audio_Reactive_Parameters.Amplitude_Music
     1=$NPC.NPC_Audio_Reactive_Parameters.Amplitude_MusicHistory0
     2=$NPC.NPC_Audio_Reactive_Parameters.Amplitude_MusicHistory1
     3=$NPC.NPC_Audio_Reactive_Parameters.Amplitude_MusicHistory2
     Index=$Particles.PartitionParticles.Partition
     Num=::NewEnumerator2
    )
    Module.Param0WriteEnabled=true
    Module.Param1WriteEnabled=true
    Module.Param2WriteEnabled=true
    Module.Param3WriteEnabled=true
   )
   /SetVariables(Particles.Position=&MergeHelper_VectorSwitch(0=<100 0 200>, 1=<75 0 200>, 2=<50 0 200>, 3=<25 0 200>, Index=$Particles.PartitionParticles.Partition, Num=::NewEnumerator2)) {isSetParameters=true, module=SetVariables_#7}
  ]
  rendererClasses [&NiagaraSpriteRendererProperties]
  renderers [{rendererClass=&NiagaraSpriteRendererProperties}]
 }
 Heart {
  emitterUpdateScript [
   &EmitterState(Life Cycle Mode=::NewEnumerator1, Loop Behavior=::NewEnumerator1, Loop Duration Mode=::NewEnumerator1, UseLoopDelay=false)
   &SpawnBurst_Instantaneous(Use Spawn Probability=false, UseLoopCountLimit=false)
  ]
  particleSpawnScript [
   &InitializeParticle(Color=$System.TODColor, Mesh Scale Mode=::NewEnumerator3, Mesh Uniform Scale=0.06, UsePositionOffset=false, Write Material Random=true)
   &CameraOffset(Camera Offset Amount=40)
   &InitialMeshOrientation(Mesh Orientation Mode=::NewEnumerator2, Rotation Coordinate Space=::NewEnumerator2, Use Rotation Vector=true)
  ]
  particleUpdateScript [
   &ParticleState(Kill Particles When Lifetime Has Elapsed=false, Loop Particles Lifetime=true)
   &ScaleMeshSize(Scale Factor=&VectorFromFloat(Value=&Waveform("Waveform [1]"=::NewEnumerator2, "[1] Frequency"=1, "[1] Amplitude Min/Max"=<-0.4 0.3>, "[1] Scale"=$System.AudioReactivity, "[1] Bias"=1)))
   &ScaleMeshSize(Scale Factor=&VectorFromFloat(Value=&Lerp_Float(A=$System.DebugScale, B=1, Alpha=$System.Timeline001.NormalizedTime))) {module=ScaleMeshSize001}
   &SkeletalMeshLocation(
    MaskSpawnGroup=false
    Mesh Sampling Type=::NewEnumerator5
    Sampled Position Offset=<2 -7 -6>
    Skeletal Mesh=$System.Skeletal Mesh
    Socket Sampling Mode=::NewEnumerator5
   )
   &ScaleColor(
    Scale RGB=&MakeVectorFromLinearColorRGB(Color=&LerpLinearColors(
      StartColor=<rgba 1 1 1 1>
      EndColor=&RandomRangeLinearColor(Minimum=<rgba 1 0.4 0.4 1>, Maximum=<rgba 1 0.4 0.41 1>)
      LerpFactor=&Waveform("Waveform [1]"=::NewEnumerator2, "[1] Frequency"=1, "[1] Amplitude Min/Max"=<0 1.2>, "[1] Scale"=$System.AudioReactivity)
     ))
    ScaleA=true
    ScaleRGB=true
   )
   &DynamicMaterialParameters(
    Index 0 Param 4=&Waveform("Waveform [1]"=::NewEnumerator2, "[1] Frequency"=1, "[1] Amplitude Min/Max"=<-2 2>, "[1] Scale"=$System.AudioReactivity)
    Module.Index0Float0Write=true
    Module.Index0Float1Write=true
    Module.Index0Float2Write=true
    Module.Index0Float3Write=true
   )
  ]
  rendererClasses [&NiagaraMeshRendererProperties]
  renderers [{rendererClass=&NiagaraMeshRendererProperties, rendererIndex=0}]
 }
 LeftArm001 {
  bEnabled=false
  emitterUpdateScript [
   &EmitterState(
    Inactive Response=::NewEnumerator1
    Life Cycle Mode=::NewEnumerator1
    Loop Behavior=::NewEnumerator1
    Loop Duration Mode=::NewEnumerator1
    UseLoopDelay=false
   )
   &SpawnRate(SpawnRate=6, Use Spawn Probability=false) {enabled=false}
   &SpawnBurst_Instantaneous(Spawn Count=40, Use Spawn Probability=false, UseLoopCountLimit=false)
  ]
  particleSpawnScript [
   &InitializeParticle(
    Color Mode=::NewEnumerator0
    Lifetime=5
    Mesh Scale Mode=::NewEnumerator3
    Mesh Uniform Scale=0.14
    Ribbon Width=7
    Ribbon Width Mode=::NewEnumerator1
    Sprite Size Mode=::NewEnumerator3
    UsePositionOffset=false
    Write Material Random=true
   )
   &CameraOffset()
  ]
  particleUpdateScript [
   &ParticleState(Kill Particles When Lifetime Has Elapsed=false, Let Infinitely Lived Particles Die When Emitter Deactivates=true)
   /Game/Effects/Niagara/Modules/Curves/FN_ListCurve_SkeletalMeshSampler(
    Apply Position=true
    Curve Index:NiagaraFloat=&ReturnNormalizedExecIndex()
    Index_01=13
    Index_02=14
    Index_03=15
    Index_04=15
    Index_05=16
    Index_06=16
    Index_07=17
    Index_08=18
    Number of Inputs=6
    Offset Coordinate Space=3
    Offset_01:Vector3f=<3.0956 0.1658 -1.5057>
    Offset_02:Vector3f=<0 2.8388 1.054>
    Offset_03:Vector3f=<0 3.2713 5.1751>
    Offset_04:Vector3f=<5.749 -1.2073 11.209>
    Offset_05:Vector3f=<3.2356 -4.247 4.4882>
    Offset_06:Vector3f=<0.0547 -10 1.0556>
    Offset_07:Vector3f=<8.6312 -16 -3>
    Offset_08:Vector3f=<2.4944 -7 0>
    Selection Type=0
    Skeletal Mesh:NiagaraDataInterfaceSkeletalMesh=$System.Skeletal Mesh
   )
   &ScaleMeshSize(Scale Factor=&VectorFromFloat(Value=&FloatFromCurve(FloatCurve=DataInterface() {propertyValues {Curve {
         defaultValue=0
         keys [
          {interpMode=RCIM_Cubic, tangentMode=RCTM_User, time=0, value=0.5}
          {interpMode=RCIM_Cubic, tangentMode=RCTM_User, time=0.5, value=0.7}
          {tangentMode=RCTM_User, time=1, value=0.5}
         ]
        }}}))) {enabled=false}
   &ScaleMeshSize(Scale Factor=&VectorFromFloat(Value=&Waveform(
      "Waveform [1]"=::NewEnumerator2
      "[1] Frequency"=1
      "[1] Amplitude Min/Max"=<0.7 1>
      "[1] Scale"=&FloatFromCurve(
       FloatCurve=DataInterface() {propertyValues {Curve {
          defaultValue=1
          keys [
           {interpMode=RCIM_Cubic, tangentMode=RCTM_User, time=0, value=0.25}
           {interpMode=RCIM_Cubic, tangentMode=RCTM_User, time=0.5, value=1}
           {tangentMode=RCTM_User, time=1, value=0.25}
          ]
         }}}
       Scale Curve=1.5
      )
     ))) {enabled=false, module=ScaleMeshSize001}
  ]
  rendererClasses [
   &NiagaraRibbonRendererProperties
   &NiagaraSpriteRendererProperties
   &NiagaraMeshRendererProperties
  ]
  renderers [
   {rendererClass=&NiagaraRibbonRendererProperties}
   {rendererClass=&NiagaraSpriteRendererProperties, rendererIndex=1}
   {rendererClass=&NiagaraMeshRendererProperties, rendererIndex=2}
  ]
 }
 LeftLeg {
  bEnabled=false
  emitterUpdateScript [
   &EmitterState(
    Inactive Response=::NewEnumerator1
    Life Cycle Mode=::NewEnumerator1
    Loop Behavior=::NewEnumerator1
    Loop Duration Mode=::NewEnumerator1
    UseLoopDelay=false
   )
   &SpawnRate(SpawnRate=6, Use Spawn Probability=false) {enabled=false}
   &SpawnBurst_Instantaneous(Spawn Count=40, Use Spawn Probability=false, UseLoopCountLimit=false)
  ]
  particleSpawnScript [
   &InitializeParticle(
    Color Mode=::NewEnumerator0
    Lifetime=5
    Mesh Scale Mode=::NewEnumerator3
    Mesh Uniform Scale=0.14
    Ribbon Width=7
    Ribbon Width Mode=::NewEnumerator1
    Sprite Size Mode=::NewEnumerator3
    UsePositionOffset=false
    Write Material Random=true
   )
   &CameraOffset()
  ]
  particleUpdateScript [
   &ParticleState(Kill Particles When Lifetime Has Elapsed=false, Let Infinitely Lived Particles Die When Emitter Deactivates=true)
   &FN_ListCurve_SkeletalMeshSampler(
    Apply Position=true
    Curve Index:NiagaraFloat=&ReturnNormalizedExecIndex()
    Index_01=8
    Index_02=8
    Index_03=8
    Index_04=8
    Index_05=9
    Index_06=9
    Number of Inputs=4
    Offset Coordinate Space=3
    Offset_01:Vector3f=<-8 -2 -8>
    Offset_02:Vector3f=<-2.2935 -3.2672 -3>
    Offset_03:Vector3f=<-0.6591 -5 0>
    Offset_04:Vector3f=<-1.7967 -10.4385 2.5>
    Offset_05:Vector3f=<-7.0153 -9.4834 -0.5925>
    Offset_06:Vector3f=<-9.1365 -9.8769 2.4868>
    Selection Type=0
    Skeletal Mesh:NiagaraDataInterfaceSkeletalMesh=$System.Skeletal Mesh
   )
   &ScaleMeshSize(Scale Factor=&VectorFromFloat(Value=&FloatFromCurve(FloatCurve=DataInterface() {propertyValues {Curve {
         defaultValue=0
         keys [
          {interpMode=RCIM_Cubic, tangentMode=RCTM_User, time=0, value=0.5}
          {interpMode=RCIM_Cubic, tangentMode=RCTM_User, time=0.5, value=0.7}
          {tangentMode=RCTM_User, time=1, value=0.5}
         ]
        }}}))) {enabled=false}
   &ScaleMeshSize(Scale Factor=&VectorFromFloat(Value=&Waveform(
      "Waveform [1]"=::NewEnumerator2
      "[1] Frequency"=1
      "[1] Amplitude Min/Max"=<0.7 1>
      "[1] Scale"=&FloatFromCurve(
       FloatCurve=DataInterface() {propertyValues {Curve {
          defaultValue=1
          keys [
           {interpMode=RCIM_Cubic, tangentMode=RCTM_User, time=0, value=0.25}
           {interpMode=RCIM_Cubic, tangentMode=RCTM_User, time=0.5, value=1}
           {tangentMode=RCTM_User, time=1, value=0.25}
          ]
         }}}
       Scale Curve=1.5
      )
     ))) {enabled=false, module=ScaleMeshSize001}
  ]
  rendererClasses [
   &NiagaraRibbonRendererProperties
   &NiagaraSpriteRendererProperties
   &NiagaraMeshRendererProperties
  ]
  renderers [
   {rendererClass=&NiagaraRibbonRendererProperties}
   {rendererClass=&NiagaraSpriteRendererProperties, rendererIndex=1}
   {rendererClass=&NiagaraMeshRendererProperties, rendererIndex=2}
  ]
 }
 Left_Arm_Voxels {
  emitterUpdateScript [
   &EmitterState(
    Inactive Response=::NewEnumerator1
    Life Cycle Mode=::NewEnumerator1
    Loop Behavior=::NewEnumerator1
    Loop Duration Mode=::NewEnumerator1
    UseLoopDelay=false
   )
   &SpawnRate(SpawnRate=&Lerp_Float(A=3, B=8, Alpha=$System.AudioReactivity), Use Spawn Probability=false)
  ]
  particleSpawnScript [
   &PartitionParticles(Number Of Partitions=3, Partition Mode:ENiagara_PartitionMode_v2=::NewEnumerator5)
   &InitializeParticle(
    Color=$System.TODColor
    Lifetime=&Lerp_Float(A=8, B=4, Alpha=$System.AudioReactivity)
    Mesh Scale=&VectorFromFloat(Value=&RandomRangeFloat(Minimum=0.09, Maximum=0.14))
    Mesh Scale Mode=::NewEnumerator4
    Position Mode=::NewEnumerator0
    Ribbon Width=7
    Ribbon Width Mode=::NewEnumerator1
    Sprite Size Mode=::NewEnumerator3
    Write Material Random=true
   )
   &InitialMeshOrientation(
    Mesh Orientation Mode=::NewEnumerator2
    Rotation=&RandomRangeVector(Minimum=<0 -0.05 -0.05>, Maximum=<1 0.05 0.05>)
    Use Rotation Vector=true
   )
   &CameraOffset(Camera Offset Amount=5)
  ]
  particleUpdateScript [
   &ParticleState()
   &Timeline(Play Rate=&Lerp_Float(A=0.15, B=0.4, Alpha=$System.AudioReactivity))
   &FN_ListCurve_SkeletalMeshSampler(
    Apply Position=true
    Curve Index:NiagaraFloat=&FloatFromCurve(
     FloatCurve=DataInterface() {propertyValues {Curve {
        defaultValue=1
        keys [{interpMode=RCIM_Cubic, tangentMode=RCTM_User, time=0, value=0}, {interpMode=RCIM_Cubic, tangentMode=RCTM_User, time=1, value=1}]
       }}}
     CurveIndex=$Particles.Timeline.NormalizedTime
    )
    Index_01=13
    Index_02=14
    Index_03=15
    Index_04=15
    Index_05=16
    Index_06=16
    Index_07=17
    Index_08=18
    "Interpolate Mesh Sampling (Update)"=true
    Number of Inputs=6
    Offset Coordinate Space=3
    Offset_01:Vector3f=&AddVectorToPosition(Position=<3.0956 -5 -1.5057>, Vector=&RandomRangeVector(Minimum=<-2 -1 -2>, Maximum=<1 1 2>))
    Offset_02:Vector3f=&AddVectorToPosition(
     Position=<0 2.8388 -1.3319>
     Vector=&VectorFromFloat(Value=&RandomRangeFloat(Minimum=-1, Maximum=1))
    )
    Offset_03:Vector3f=&AddVectorToPosition(
     Position=<0 3.2713 5.1751>
     Vector=&VectorFromFloat(Value=&RandomRangeFloat(Minimum=-1.25, Maximum=-1.25))
    )
    Offset_04:Vector3f=&AddVectorToPosition(
     Position=<5.749 -1.2073 11.209>
     Vector=&VectorFromFloat(Value=&RandomRangeFloat(Minimum=-1.25, Maximum=-1.25))
    )
    Offset_05:Vector3f=&AddVectorToPosition(
     Position=<3.2356 -4.247 4.4882>
     Vector=&VectorFromFloat(Value=&RandomRangeFloat(Minimum=-1.5, Maximum=-1.5))
    )
    Offset_06:Vector3f=&AddVectorToPosition(
     Position=<0.0547 -10 1.0556>
     Vector=&VectorFromFloat(Value=&RandomRangeFloat(Minimum=-1.5, Maximum=-1.5))
    )
    Offset_07:Vector3f=&AddVectorToPosition(
     Position=<8.6312 -16 -3>
     Vector=&VectorFromFloat(Value=&RandomRangeFloat(Minimum=-1.5, Maximum=-1.5))
    )
    Offset_08:Vector3f=&AddVectorToPosition(Position=<2.4944 -7 0>, Vector=&RandomRangeVector(Minimum=<-5 -5 -5>, Maximum=<5 2 5>))
    Selection Type=0
    Skeletal Mesh:NiagaraDataInterfaceSkeletalMesh=$System.Skeletal Mesh
   ) {module=FN_ListCurve_SkeletalMeshSampler001}
   &ScaleMeshSize(Scale Factor=&VectorFromFloat(Value=&FloatFromCurve(
      FloatCurve=DataInterface() {propertyValues {Curve {
         defaultValue=0
         keys [
          {arriveTangent=1.098, interpMode=RCIM_Cubic, leaveTangent=1.098, tangentMode=RCTM_User, time=0, value=0.2}
          {arriveTangent=-0.0768, interpMode=RCIM_Cubic, leaveTangent=-0.0768, tangentMode=RCTM_User, time=0.5, value=0.7}
          {arriveTangent=-0.1845, tangentMode=RCTM_User, time=1, value=0}
         ]
        }}}
      CurveIndex=$Particles.Timeline.NormalizedTime
      Scale Curve=&Lerp_Float(
       A=&Lerp_Float(A=$System.DebugScale, B=1, Alpha=$System.Timeline001.NormalizedTime)
       B=&Lerp_Float(A=1.2, B=0.9, Alpha=$System.Timeline.NormalizedTime)
       Alpha=$System.AudioReactivity
      )
     )))
   &ScaleMeshSize(Scale Factor=&VectorFromFloat(Value=&MergeHelper_FloatSwitch(
      0=&Waveform(
       "Waveform [1]"=::NewEnumerator2
       "[1] Frequency"=1
       "[1] Amplitude Min/Max"=<-0.25 0.75>
       "[1] Scale"=$System.AudioLow
       "[1] Bias"=&RandomRangeFloat(Minimum=0.8, Maximum=1.1)
      )
      1=&Waveform(
       "Waveform [1]"=::NewEnumerator2
       "[1] Frequency"=1.5
       "[1] Amplitude Min/Max"=<-0.25 0.75>
       "[1] Scale"=$System.AudioMedium
       "[1] Bias"=&RandomRangeFloat(Minimum=0.8, Maximum=1.1)
      )
      2=&Waveform(
       "Waveform [1]"=::NewEnumerator2
       "[1] Frequency"=2
       "[1] Amplitude Min/Max"=<-0.25 0.75>
       "[1] Scale"=$System.AudioHigh
       "[1] Bias"=&RandomRangeFloat(Minimum=0.8, Maximum=1.1)
      )
      3=?
      Index=$Particles.PartitionParticles.Partition
      Num=::NewEnumerator1
     ))) {module=ScaleMeshSize003}
   &ScaleMeshSize(Scale Factor=&MakeVector(
     X=&Waveform("Waveform [1]"=::NewEnumerator2, "[1] Frequency"=1, "[1] Amplitude Min/Max"=<-0.25 0.25>, "[1] Scale"=$System.AudioLow, "[1] Bias"=1)
     Y=&Waveform("Waveform [1]"=::NewEnumerator2, "[1] Frequency"=2, "[1] Amplitude Min/Max"=<-0.25 0.25>, "[1] Scale"=$System.AudioMedium, "[1] Bias"=1)
     Z=&Waveform("Waveform [1]"=::NewEnumerator2, "[1] Frequency"=3, "[1] Amplitude Min/Max"=<-0.25 0.25>, "[1] Scale"=$System.AudioHigh, "[1] Bias"=1)
    )) {module=ScaleMeshSize001}
   &ScaleColor(
    Scale RGB=&MakeVectorFromLinearColorRGB(Color=&MergeHelper_ColorSwitch(
      0=&LerpLinearColors(
       StartColor=<rgba 1 1 1 1>
       EndColor=&RandomRangeLinearColor(Minimum=<rgba 1 0.4 0.4 1>, Maximum=<rgba 0.4 1 0.4 1>)
       LerpFactor=&Waveform("Waveform [1]"=::NewEnumerator2, "[1] Frequency"=1, "[1] Amplitude Min/Max"=<0 1.2>, "[1] Scale"=$System.AudioLow)
      )
      1=&LerpLinearColors(
       StartColor=<rgba 1 1 1 1>
       EndColor=&RandomRangeLinearColor(Minimum=<rgba 0.4 1 0.4 1>, Maximum=<rgba 0.4 0.4 1 1>)
       LerpFactor=&Waveform("Waveform [1]"=::NewEnumerator2, "[1] Frequency"=1.5, "[1] Amplitude Min/Max"=<0 1.2>, "[1] Scale"=$System.AudioMedium)
      )
      2=&LerpLinearColors(
       StartColor=<rgba 1 1 1 1>
       EndColor=&RandomRangeLinearColor(Minimum=<rgba 0.4 0.4 1 1>, Maximum=<rgba 1 0.4 0.4 1>)
       LerpFactor=&Waveform("Waveform [1]"=::NewEnumerator2, "[1] Frequency"=2, "[1] Amplitude Min/Max"=<0 1.2>, "[1] Scale"=$System.AudioHigh)
      )
      3=?
      Index=$Particles.PartitionParticles.Partition
      Num=::NewEnumerator1
     ))
    ScaleA=true
    ScaleRGB=true
   ) {module=ScaleColor001}
   &SolveForcesAndVelocity(Clamp Velocity=false, Limit Acceleration=false)
   &OffsetPosition(
    Interpolation Curve=DataInterface() {propertyValues {Curve {
       defaultValue=0
       keys [{interpMode=RCIM_Cubic, tangentMode=RCTM_User, time=0, value=0}, {interpMode=RCIM_Cubic, tangentMode=RCTM_User, time=1, value=1}]
      }}}
    Interpolation Scale:NiagaraFloat=$System.AudioReactivity
    Jitter Amount=0.4
    Jitter Interpolation=2
    Jitter Time Offset=0.05
    Offset Mode=1
   )
   &DynamicMaterialParameters(
    Index 0 Param 4=&MergeHelper_FloatSwitch(
     0=&Waveform("Waveform [1]"=::NewEnumerator2, "[1] Frequency"=1, "[1] Amplitude Min/Max"=<0 1>, "[1] Scale"=$System.AudioLow)
     1=&Waveform("Waveform [1]"=::NewEnumerator2, "[1] Frequency"=1.5, "[1] Amplitude Min/Max"=<0 1>, "[1] Scale"=$System.AudioMedium)
     2=&Waveform("Waveform [1]"=::NewEnumerator2, "[1] Frequency"=2, "[1] Amplitude Min/Max"=<0 1>, "[1] Scale"=$System.AudioHigh)
     3=?
     Index=$Particles.PartitionParticles.Partition
     Num=::NewEnumerator1
    )
    Module.Index0Float0Write=true
    Module.Index0Float1Write=true
    Module.Index0Float2Write=true
    Module.Index0Float3Write=true
   )
  ]
  rendererClasses [&NiagaraMeshRendererProperties]
  renderers [{rendererClass=&NiagaraMeshRendererProperties, rendererIndex=0}]
 }
 Left_Arm_Voxels001 {
  bEnabled=false
  emitterUpdateScript [
   &EmitterState(
    Inactive Response=::NewEnumerator1
    Life Cycle Mode=::NewEnumerator1
    Loop Behavior=::NewEnumerator1
    Loop Duration Mode=::NewEnumerator1
    UseLoopDelay=false
   )
   &SpawnRate(SpawnRate=&Lerp_Float(A=3, B=8, Alpha=$System.AudioReactivity), Use Spawn Probability=false)
  ]
  particleSpawnScript [
   &PartitionParticles(Number Of Partitions=3, Partition Mode:ENiagara_PartitionMode_v2=::NewEnumerator5)
   &InitializeParticle(
    Color=$System.TODColor
    Lifetime=&Lerp_Float(A=8, B=4, Alpha=$System.AudioReactivity)
    Mesh Scale=&VectorFromFloat(Value=&RandomRangeFloat(Minimum=0.09, Maximum=0.14))
    Mesh Scale Mode=::NewEnumerator4
    Position Mode=::NewEnumerator0
    Ribbon Width=7
    Ribbon Width Mode=::NewEnumerator1
    Sprite Size Mode=::NewEnumerator3
    Write Material Random=true
   )
   &InitialMeshOrientation(
    Mesh Orientation Mode=::NewEnumerator2
    Rotation=&RandomRangeVector(Minimum=<0 -0.05 -0.05>, Maximum=<1 0.05 0.05>)
    Use Rotation Vector=true
   )
   &CameraOffset(Camera Offset Amount=5)
  ]
  particleUpdateScript [
   &ParticleState()
   &Timeline(Play Rate=&Lerp_Float(A=0.15, B=0.4, Alpha=$System.AudioReactivity))
   &FN_ListCurve_SkeletalMeshSampler(
    Apply Position=true
    Curve Index:NiagaraFloat=&FloatFromCurve(
     FloatCurve=DataInterface() {propertyValues {Curve {
        defaultValue=1
        keys [{interpMode=RCIM_Cubic, tangentMode=RCTM_User, time=0, value=0}, {interpMode=RCIM_Cubic, tangentMode=RCTM_User, time=1, value=1}]
       }}}
     CurveIndex=$Particles.Timeline.NormalizedTime
    )
    Index_01=13
    Index_02=14
    Index_03=15
    Index_04=15
    Index_05=16
    Index_06=16
    Index_07=17
    Index_08=18
    "Interpolate Mesh Sampling (Update)"=true
    Number of Inputs=6
    Offset Coordinate Space=3
    Offset_01:Vector3f=&AddVectorToPosition(Position=<3.0956 -5 -1.5057>, Vector=&RandomRangeVector(Minimum=<-2 -1 -2>, Maximum=<1 1 2>))
    Offset_02:Vector3f=&AddVectorToPosition(
     Position=<0 2.8388 -1.3319>
     Vector=&VectorFromFloat(Value=&RandomRangeFloat(Minimum=-1, Maximum=1))
    )
    Offset_03:Vector3f=&AddVectorToPosition(
     Position=<0 3.2713 5.1751>
     Vector=&VectorFromFloat(Value=&RandomRangeFloat(Minimum=-1.25, Maximum=-1.25))
    )
    Offset_04:Vector3f=&AddVectorToPosition(
     Position=<5.749 -1.2073 11.209>
     Vector=&VectorFromFloat(Value=&RandomRangeFloat(Minimum=-1.25, Maximum=-1.25))
    )
    Offset_05:Vector3f=&AddVectorToPosition(
     Position=<3.2356 -4.247 4.4882>
     Vector=&VectorFromFloat(Value=&RandomRangeFloat(Minimum=-1.5, Maximum=-1.5))
    )
    Offset_06:Vector3f=&AddVectorToPosition(
     Position=<0.0547 -10 1.0556>
     Vector=&VectorFromFloat(Value=&RandomRangeFloat(Minimum=-1.5, Maximum=-1.5))
    )
    Offset_07:Vector3f=&AddVectorToPosition(
     Position=<8.6312 -16 -3>
     Vector=&VectorFromFloat(Value=&RandomRangeFloat(Minimum=-1.5, Maximum=-1.5))
    )
    Offset_08:Vector3f=&AddVectorToPosition(Position=<2.4944 -7 0>, Vector=&RandomRangeVector(Minimum=<-5 -5 -5>, Maximum=<5 2 5>))
    Selection Type=0
    Skeletal Mesh:NiagaraDataInterfaceSkeletalMesh=$System.Skeletal Mesh
   ) {module=FN_ListCurve_SkeletalMeshSampler001}
   &ScaleMeshSize(Scale Factor=&VectorFromFloat(Value=&FloatFromCurve(
      FloatCurve=DataInterface() {propertyValues {Curve {
         defaultValue=0
         keys [
          {arriveTangent=1.098, interpMode=RCIM_Cubic, leaveTangent=1.098, tangentMode=RCTM_User, time=0, value=0.2}
          {arriveTangent=-0.0768, interpMode=RCIM_Cubic, leaveTangent=-0.0768, tangentMode=RCTM_User, time=0.5, value=0.7}
          {arriveTangent=-0.1845, tangentMode=RCTM_User, time=1, value=0}
         ]
        }}}
      CurveIndex=$Particles.Timeline.NormalizedTime
      Scale Curve=&Lerp_Float(
       A=&Lerp_Float(A=$System.DebugScale, B=1, Alpha=$System.Timeline001.NormalizedTime)
       B=&Lerp_Float(A=1.2, B=0.9, Alpha=$System.Timeline.NormalizedTime)
       Alpha=$System.AudioReactivity
      )
     )))
   &ScaleMeshSize(Scale Factor=&VectorFromFloat(Value=&MergeHelper_FloatSwitch(
      0=&Waveform(
       "Waveform [1]"=::NewEnumerator2
       "[1] Frequency"=1
       "[1] Amplitude Min/Max"=<-0.25 0.75>
       "[1] Scale"=$System.AudioLow
       "[1] Bias"=&RandomRangeFloat(Minimum=0.8, Maximum=1.1)
      )
      1=&Waveform(
       "Waveform [1]"=::NewEnumerator2
       "[1] Frequency"=1.5
       "[1] Amplitude Min/Max"=<-0.25 0.75>
       "[1] Scale"=$System.AudioMedium
       "[1] Bias"=&RandomRangeFloat(Minimum=0.8, Maximum=1.1)
      )
      2=&Waveform(
       "Waveform [1]"=::NewEnumerator2
       "[1] Frequency"=2
       "[1] Amplitude Min/Max"=<-0.25 0.75>
       "[1] Scale"=$System.AudioHigh
       "[1] Bias"=&RandomRangeFloat(Minimum=0.8, Maximum=1.1)
      )
      3=?
      Index=$Particles.PartitionParticles.Partition
      Num=::NewEnumerator1
     ))) {module=ScaleMeshSize003}
   &ScaleMeshSize(Scale Factor=&MakeVector(
     X=&Waveform("Waveform [1]"=::NewEnumerator2, "[1] Frequency"=1, "[1] Amplitude Min/Max"=<-0.25 0.25>, "[1] Scale"=$System.AudioLow, "[1] Bias"=1)
     Y=&Waveform("Waveform [1]"=::NewEnumerator2, "[1] Frequency"=2, "[1] Amplitude Min/Max"=<-0.25 0.25>, "[1] Scale"=$System.AudioMedium, "[1] Bias"=1)
     Z=&Waveform("Waveform [1]"=::NewEnumerator2, "[1] Frequency"=3, "[1] Amplitude Min/Max"=<-0.25 0.25>, "[1] Scale"=$System.AudioHigh, "[1] Bias"=1)
    )) {module=ScaleMeshSize001}
   &ScaleColor(
    Scale RGB=&MakeVectorFromLinearColorRGB(Color=&MergeHelper_ColorSwitch(
      0=&LerpLinearColors(
       StartColor=<rgba 1 1 1 1>
       EndColor=&RandomRangeLinearColor(Minimum=<rgba 1 0.4 0.4 1>, Maximum=<rgba 0.4 1 0.4 1>)
       LerpFactor=&Waveform("Waveform [1]"=::NewEnumerator2, "[1] Frequency"=1, "[1] Amplitude Min/Max"=<0 1.2>, "[1] Scale"=$System.AudioLow)
      )
      1=&LerpLinearColors(
       StartColor=<rgba 1 1 1 1>
       EndColor=&RandomRangeLinearColor(Minimum=<rgba 0.4 1 0.4 1>, Maximum=<rgba 0.4 0.4 1 1>)
       LerpFactor=&Waveform("Waveform [1]"=::NewEnumerator2, "[1] Frequency"=1.5, "[1] Amplitude Min/Max"=<0 1.2>, "[1] Scale"=$System.AudioMedium)
      )
      2=&LerpLinearColors(
       StartColor=<rgba 1 1 1 1>
       EndColor=&RandomRangeLinearColor(Minimum=<rgba 0.4 0.4 1 1>, Maximum=<rgba 1 0.4 0.4 1>)
       LerpFactor=&Waveform("Waveform [1]"=::NewEnumerator2, "[1] Frequency"=2, "[1] Amplitude Min/Max"=<0 1.2>, "[1] Scale"=$System.AudioHigh)
      )
      3=?
      Index=$Particles.PartitionParticles.Partition
      Num=::NewEnumerator1
     ))
    ScaleA=true
    ScaleRGB=true
   ) {module=ScaleColor001}
   &OffsetPosition(
    Interpolation Curve=DataInterface() {propertyValues {Curve {
       defaultValue=0
       keys [{interpMode=RCIM_Cubic, tangentMode=RCTM_User, time=0, value=0}, {interpMode=RCIM_Cubic, tangentMode=RCTM_User, time=1, value=1}]
      }}}
    Interpolation Scale:NiagaraFloat=$System.AudioReactivity
    Jitter Amount=0.4
    Jitter Interpolation=2
    Jitter Time Offset=0.05
    Offset Mode=1
   )
   &DynamicMaterialParameters(
    Index 0 Param 4=&MergeHelper_FloatSwitch(
     0=&Waveform("Waveform [1]"=::NewEnumerator2, "[1] Frequency"=1, "[1] Amplitude Min/Max"=<0 1>, "[1] Scale"=$System.AudioLow)
     1=&Waveform("Waveform [1]"=::NewEnumerator2, "[1] Frequency"=1.5, "[1] Amplitude Min/Max"=<0 1>, "[1] Scale"=$System.AudioMedium)
     2=&Waveform("Waveform [1]"=::NewEnumerator2, "[1] Frequency"=2, "[1] Amplitude Min/Max"=<0 1>, "[1] Scale"=$System.AudioHigh)
     3=?
     Index=$Particles.PartitionParticles.Partition
     Num=::NewEnumerator1
    )
    Module.Index0Float0Write=true
    Module.Index0Float1Write=true
    Module.Index0Float2Write=true
    Module.Index0Float3Write=true
   )
  ]
  rendererClasses [&NiagaraMeshRendererProperties]
  renderers [{rendererClass=&NiagaraMeshRendererProperties, rendererIndex=0}]
 }
 Left_Leg_Voxels {
  emitterUpdateScript [
   &EmitterState(
    Inactive Response=::NewEnumerator1
    Life Cycle Mode=::NewEnumerator1
    Loop Behavior=::NewEnumerator1
    Loop Duration Mode=::NewEnumerator1
    UseLoopDelay=false
   )
   &SpawnRate(SpawnRate=&Lerp_Float(A=2, B=5, Alpha=$System.AudioReactivity), Use Spawn Probability=false)
  ]
  particleSpawnScript [
   &PartitionParticles(Number Of Partitions=3, Partition Mode:ENiagara_PartitionMode_v2=::NewEnumerator5)
   &InitializeParticle(
    Color=$System.TODColor
    Lifetime=&Lerp_Float(A=6, B=3, Alpha=$System.AudioReactivity)
    Mesh Scale=&VectorFromFloat(Value=&RandomRangeFloat(Minimum=0.09, Maximum=0.14))
    Mesh Scale Mode=::NewEnumerator4
    Position Mode=::NewEnumerator0
    Ribbon Width=7
    Ribbon Width Mode=::NewEnumerator1
    Sprite Size Mode=::NewEnumerator3
    Write Material Random=true
   )
   &InitialMeshOrientation(
    Mesh Orientation Mode=::NewEnumerator2
    Rotation=&RandomRangeVector(Minimum=<0 -0.05 -0.05>, Maximum=<1 0.05 0.05>)
    Use Rotation Vector=true
   )
  ]
  particleUpdateScript [
   &ParticleState()
   &Timeline(Play Rate=&Lerp_Float(A=0.2, B=0.4, Alpha=$System.AudioReactivity))
   &FN_ListCurve_SkeletalMeshSampler(
    Apply Orientation=true
    Apply Position=true
    Bone Rotation Contribution=0
    Curve Index:NiagaraFloat=&FloatFromCurve(FloatCurve=DataInterface() {propertyValues {Curve {
        defaultValue=1
        keys [{interpMode=RCIM_Cubic, tangentMode=RCTM_User, time=0, value=0}, {interpMode=RCIM_Cubic, tangentMode=RCTM_User, time=1, value=1}]
       }}})
    Index_01=8
    Index_02=8
    Index_03=8
    Index_04=8
    Index_05=9
    Index_06=9
    Number of Inputs=4
    Offset Coordinate Space=3
    Offset_01:Vector3f=&AddVectorToPosition(Position=<-8 -2 -8>, Vector=&RandomRangeVector(Minimum=<-2 -2 -2>, Maximum=<2 2 2>))
    Offset_02:Vector3f=&AddVectorToPosition(
     Position=<-2.2935 -3.2672 -3>
     Vector=&VectorFromFloat(Value=&RandomRangeFloat(Minimum=-1, Maximum=1))
    )
    Offset_03:Vector3f=&AddVectorToPosition(
     Position=<-0.6591 -5 0>
     Vector=&VectorFromFloat(Value=&RandomRangeFloat(Minimum=-1.5, Maximum=1.5))
    )
    Offset_04:Vector3f=&AddVectorToPosition(
     Position=<-1.7967 -10.4385 2.5>
     Vector=&VectorFromFloat(Value=&RandomRangeFloat(Minimum=-1, Maximum=1))
    )
    Offset_05:Vector3f=&AddVectorToPosition(
     Position=<-7.0153 -9.4834 -0.5925>
     Vector=&VectorFromFloat(Value=&RandomRangeFloat(Minimum=-2, Maximum=2))
    )
    Offset_06:Vector3f=&AddVectorToPosition(
     Position=<-9.1365 -9.8769 2.4868>
     Vector=&VectorFromFloat(Value=&RandomRangeFloat(Minimum=-4, Maximum=4))
    )
    Selection Type=0
    Skeletal Mesh:NiagaraDataInterfaceSkeletalMesh=$System.Skeletal Mesh
   ) {module=FN_ListCurve_SkeletalMeshSampler001}
   &ScaleMeshSize(Scale Factor=&VectorFromFloat(Value=&FloatFromCurve(
      FloatCurve=DataInterface() {propertyValues {Curve {
         defaultValue=0
         keys [
          {arriveTangent=1.098, interpMode=RCIM_Cubic, leaveTangent=1.098, tangentMode=RCTM_User, time=0, value=0.2}
          {arriveTangent=-0.245, interpMode=RCIM_Cubic, leaveTangent=-0.245, tangentMode=RCTM_User, time=0.5, value=0.7}
          {arriveTangent=0.0898, tangentMode=RCTM_User, time=1, value=0}
         ]
        }}}
      CurveIndex=$Particles.Timeline.NormalizedTime
      Scale Curve=&Lerp_Float(A=$System.DebugScale, B=1.2, Alpha=$System.Timeline001.NormalizedTime)
     )))
   &ScaleMeshSize(Scale Factor=&VectorFromFloat(Value=&MergeHelper_FloatSwitch(
      0=&Waveform(
       "Waveform [1]"=::NewEnumerator2
       "[1] Frequency"=1
       "[1] Amplitude Min/Max"=<-0.25 0.75>
       "[1] Scale"=$System.AudioLow
       "[1] Bias"=&RandomRangeFloat(Minimum=0.8, Maximum=1.1)
      )
      1=&Waveform(
       "Waveform [1]"=::NewEnumerator2
       "[1] Frequency"=1.5
       "[1] Amplitude Min/Max"=<-0.25 0.75>
       "[1] Scale"=$System.AudioMedium
       "[1] Bias"=&RandomRangeFloat(Minimum=0.8, Maximum=1.1)
      )
      2=&Waveform(
       "Waveform [1]"=::NewEnumerator2
       "[1] Frequency"=2
       "[1] Amplitude Min/Max"=<-0.25 0.75>
       "[1] Scale"=$System.AudioHigh
       "[1] Bias"=&RandomRangeFloat(Minimum=0.8, Maximum=1.1)
      )
      3=?
      Index=$Particles.PartitionParticles.Partition
      Num=::NewEnumerator1
     ))) {module=ScaleMeshSize003}
   &ScaleMeshSize(Scale Factor=&MakeVector(
     X=&Waveform("Waveform [1]"=::NewEnumerator2, "[1] Frequency"=1, "[1] Amplitude Min/Max"=<-0.25 0.25>, "[1] Scale"=$System.AudioLow, "[1] Bias"=1)
     Y=&Waveform("Waveform [1]"=::NewEnumerator2, "[1] Frequency"=2, "[1] Amplitude Min/Max"=<-0.25 0.25>, "[1] Scale"=$System.AudioMedium, "[1] Bias"=1)
     Z=&Waveform("Waveform [1]"=::NewEnumerator2, "[1] Frequency"=3, "[1] Amplitude Min/Max"=<-0.25 0.25>, "[1] Scale"=$System.AudioHigh, "[1] Bias"=1)
    )) {module=ScaleMeshSize001}
   &ScaleColor(
    Scale RGB=&MakeVectorFromLinearColorRGB(Color=&MergeHelper_ColorSwitch(
      0=&LerpLinearColors(
       StartColor=<rgba 1 1 1 1>
       EndColor=&RandomRangeLinearColor(Minimum=<rgba 1 0.4 0.4 1>, Maximum=<rgba 0.4 1 0.4 1>)
       LerpFactor=&Waveform("Waveform [1]"=::NewEnumerator2, "[1] Frequency"=1, "[1] Amplitude Min/Max"=<0 1.2>, "[1] Scale"=$System.AudioLow)
      )
      1=&LerpLinearColors(
       StartColor=<rgba 1 1 1 1>
       EndColor=&RandomRangeLinearColor(Minimum=<rgba 0.4 1 0.4 1>, Maximum=<rgba 0.4 0.4 1 1>)
       LerpFactor=&Waveform("Waveform [1]"=::NewEnumerator2, "[1] Frequency"=1.5, "[1] Amplitude Min/Max"=<0 1.2>, "[1] Scale"=$System.AudioMedium)
      )
      2=&LerpLinearColors(
       StartColor=<rgba 1 1 1 1>
       EndColor=&RandomRangeLinearColor(Minimum=<rgba 0.4 0.4 1 1>, Maximum=<rgba 1 0.4 0.4 1>)
       LerpFactor=&Waveform("Waveform [1]"=::NewEnumerator2, "[1] Frequency"=2, "[1] Amplitude Min/Max"=<0 1.2>, "[1] Scale"=$System.AudioHigh)
      )
      3=?
      Index=$Particles.PartitionParticles.Partition
      Num=::NewEnumerator1
     ))
    ScaleA=true
    ScaleRGB=true
   ) {module=ScaleColor001}
   &SolveForcesAndVelocity(Clamp Velocity=false, Limit Acceleration=false)
   &OffsetPosition(Interpolation Scale:NiagaraFloat=$System.AudioReactivity, Jitter Amount=0.15, Jitter Interpolation=1, Jitter Time Offset=0.05, Offset Mode=1)
   &DynamicMaterialParameters(
    Index 0 Param 4=&MergeHelper_FloatSwitch(
     0=&Waveform("Waveform [1]"=::NewEnumerator8, "[1] Frequency"=1, "[1] Amplitude Min/Max"=<0 1>, "[1] Scale"=$System.AudioLow)
     1=&Waveform("Waveform [1]"=::NewEnumerator8, "[1] Frequency"=1.5, "[1] Amplitude Min/Max"=<0 1>, "[1] Scale"=$System.AudioMedium)
     2=&Waveform("Waveform [1]"=::NewEnumerator8, "[1] Frequency"=2, "[1] Amplitude Min/Max"=<0 1>, "[1] Scale"=$System.AudioHigh)
     3=?
     Index=$Particles.PartitionParticles.Partition
     Num=::NewEnumerator1
    )
    Module.Index0Float0Write=true
    Module.Index0Float1Write=true
    Module.Index0Float2Write=true
    Module.Index0Float3Write=true
   )
  ]
  rendererClasses [&NiagaraMeshRendererProperties]
  renderers [{rendererClass=&NiagaraMeshRendererProperties, rendererIndex=0}]
 }
 RightArm {
  bEnabled=false
  emitterUpdateScript [
   &EmitterState(
    Inactive Response=::NewEnumerator1
    Life Cycle Mode=::NewEnumerator1
    Loop Behavior=::NewEnumerator1
    Loop Duration Mode=::NewEnumerator1
    UseLoopDelay=false
   )
   &SpawnRate(SpawnRate=6, Use Spawn Probability=false) {enabled=false}
   &SpawnBurst_Instantaneous(Spawn Count=40, Use Spawn Probability=false, UseLoopCountLimit=false)
  ]
  particleSpawnScript [
   &InitializeParticle(
    Color Mode=::NewEnumerator0
    Lifetime=5
    Mesh Scale Mode=::NewEnumerator3
    Mesh Uniform Scale=0.14
    Ribbon Width=7
    Ribbon Width Mode=::NewEnumerator1
    Sprite Size Mode=::NewEnumerator3
    UsePositionOffset=false
    Write Material Random=true
   )
   &CameraOffset()
  ]
  particleUpdateScript [
   &ParticleState(Kill Particles When Lifetime Has Elapsed=false, Let Infinitely Lived Particles Die When Emitter Deactivates=true)
   &FN_ListCurve_SkeletalMeshSampler(
    Apply Position=true
    Curve Index:NiagaraFloat=&ReturnNormalizedExecIndex()
    Index_01=19
    Index_02=20
    Index_03=21
    Index_04=21
    Index_05=22
    Index_06=23
    Index_07=24
    Index_08=18
    Number of Inputs=6
    Offset Coordinate Space=3
    Offset_02:Vector3f=<0 2.6949 -0.0579>
    Offset_03:Vector3f=<0 0 8.9572>
    Offset_04:Vector3f=<0 -4.5898 13.454>
    Offset_05:Vector3f=<0 0 6.6452>
    Offset_06:Vector3f=<0 -2.0833 8.8563>
    Offset_07:Vector3f=<-4.4009 -17.1397 -1.8698>
    Offset_08:Vector3f=<-1.4753 -6.6549 -0.135>
    Selection Type=0
    Skeletal Mesh:NiagaraDataInterfaceSkeletalMesh=$System.Skeletal Mesh
   )
   &ScaleMeshSize(Scale Factor=&VectorFromFloat(Value=&FloatFromCurve(FloatCurve=DataInterface() {propertyValues {Curve {
         defaultValue=0
         keys [
          {interpMode=RCIM_Cubic, tangentMode=RCTM_User, time=0, value=0.5}
          {interpMode=RCIM_Cubic, tangentMode=RCTM_User, time=0.5, value=0.7}
          {tangentMode=RCTM_User, time=1, value=0.5}
         ]
        }}}))) {enabled=false}
   &ScaleMeshSize(Scale Factor=&VectorFromFloat(Value=&Waveform(
      "Waveform [1]"=::NewEnumerator2
      "[1] Frequency"=1
      "[1] Amplitude Min/Max"=<0.7 1>
      "[1] Scale"=&FloatFromCurve(
       FloatCurve=DataInterface() {propertyValues {Curve {
          defaultValue=1
          keys [
           {interpMode=RCIM_Cubic, tangentMode=RCTM_User, time=0, value=0.25}
           {interpMode=RCIM_Cubic, tangentMode=RCTM_User, time=0.5, value=1}
           {tangentMode=RCTM_User, time=1, value=0.25}
          ]
         }}}
       Scale Curve=1.5
      )
     ))) {enabled=false, module=ScaleMeshSize001}
  ]
  rendererClasses [
   &NiagaraRibbonRendererProperties
   &NiagaraSpriteRendererProperties
   &NiagaraMeshRendererProperties
  ]
  renderers [
   {rendererClass=&NiagaraRibbonRendererProperties}
   {rendererClass=&NiagaraSpriteRendererProperties, rendererIndex=1}
   {rendererClass=&NiagaraMeshRendererProperties, rendererIndex=2}
  ]
 }
 RightLeg {
  bEnabled=false
  emitterUpdateScript [
   &EmitterState(
    Inactive Response=::NewEnumerator1
    Life Cycle Mode=::NewEnumerator1
    Loop Behavior=::NewEnumerator1
    Loop Duration Mode=::NewEnumerator1
    UseLoopDelay=false
   )
   &SpawnRate(SpawnRate=6, Use Spawn Probability=false) {enabled=false}
   &SpawnBurst_Instantaneous(Spawn Count=40, Use Spawn Probability=false, UseLoopCountLimit=false)
  ]
  particleSpawnScript [
   &InitializeParticle(
    Color Mode=::NewEnumerator0
    Lifetime=5
    Mesh Scale Mode=::NewEnumerator3
    Mesh Uniform Scale=0.14
    Ribbon Width=7
    Ribbon Width Mode=::NewEnumerator1
    Sprite Size Mode=::NewEnumerator3
    UsePositionOffset=false
    Write Material Random=true
   )
   &CameraOffset()
  ]
  particleUpdateScript [
   &ParticleState(Kill Particles When Lifetime Has Elapsed=false, Let Infinitely Lived Particles Die When Emitter Deactivates=true)
   &FN_ListCurve_SkeletalMeshSampler(
    Apply Position=true
    Curve Index:NiagaraFloat=&ReturnNormalizedExecIndex()
    Index_01=10
    Index_02=10
    Index_03=11
    Index_04=11
    Index_05=11
    Index_06=12
    Index_07=12
    Number of Inputs=5
    Offset Coordinate Space=3
    Offset_01:Vector3f=<2.4966 5.0004 0>
    Offset_02:Vector3f=<-2.3226 8.024 1.9512>
    Offset_03:Vector3f=<4.1941 9.9182 -4.4708>
    Offset_04:Vector3f=<1.9425 8.3694 -0.3202>
    Offset_05:Vector3f=<1.2524 -0.7642 3.055>
    Offset_06:Vector3f=<-7.0519 4.3101 0.9751>
    Offset_07:Vector3f=<-0.1625 0.274 4.5005>
    Selection Type=0
    Skeletal Mesh:NiagaraDataInterfaceSkeletalMesh=$System.Skeletal Mesh
   )
   &ScaleMeshSize(Scale Factor=&VectorFromFloat(Value=&FloatFromCurve(FloatCurve=DataInterface() {propertyValues {Curve {
         defaultValue=0
         keys [
          {interpMode=RCIM_Cubic, tangentMode=RCTM_User, time=0, value=0.5}
          {interpMode=RCIM_Cubic, tangentMode=RCTM_User, time=0.5, value=0.7}
          {tangentMode=RCTM_User, time=1, value=0.5}
         ]
        }}}))) {enabled=false}
   &ScaleMeshSize(Scale Factor=&VectorFromFloat(Value=&Waveform(
      "Waveform [1]"=::NewEnumerator2
      "[1] Frequency"=1
      "[1] Amplitude Min/Max"=<0.7 1>
      "[1] Scale"=&FloatFromCurve(
       FloatCurve=DataInterface() {propertyValues {Curve {
          defaultValue=1
          keys [
           {interpMode=RCIM_Cubic, tangentMode=RCTM_User, time=0, value=0.25}
           {interpMode=RCIM_Cubic, tangentMode=RCTM_User, time=0.5, value=1}
           {tangentMode=RCTM_User, time=1, value=0.25}
          ]
         }}}
       Scale Curve=1.5
      )
     ))) {enabled=false, module=ScaleMeshSize001}
  ]
  rendererClasses [
   &NiagaraRibbonRendererProperties
   &NiagaraSpriteRendererProperties
   &NiagaraMeshRendererProperties
  ]
  renderers [
   {rendererClass=&NiagaraRibbonRendererProperties}
   {rendererClass=&NiagaraSpriteRendererProperties, rendererIndex=1}
   {rendererClass=&NiagaraMeshRendererProperties, rendererIndex=2}
  ]
 }
 Right_Arm_Voxels {
  emitterUpdateScript [
   &EmitterState(
    Inactive Response=::NewEnumerator1
    Life Cycle Mode=::NewEnumerator1
    Loop Behavior=::NewEnumerator1
    Loop Duration Mode=::NewEnumerator1
    UseLoopDelay=false
   )
   &SpawnRate(SpawnRate=&Lerp_Float(A=3, B=8, Alpha=$System.AudioReactivity), Use Spawn Probability=false)
  ]
  particleSpawnScript [
   &PartitionParticles(Number Of Partitions=3, Partition Mode:ENiagara_PartitionMode_v2=::NewEnumerator5)
   &InitializeParticle(
    Color=$System.TODColor
    Lifetime=&Lerp_Float(A=8, B=4, Alpha=$System.AudioReactivity)
    Mesh Scale=&VectorFromFloat(Value=&RandomRangeFloat(Minimum=0.09, Maximum=0.14))
    Mesh Scale Mode=::NewEnumerator4
    Position Mode=::NewEnumerator0
    Ribbon Width=7
    Ribbon Width Mode=::NewEnumerator1
    Sprite Size Mode=::NewEnumerator3
    Write Material Random=true
   )
   &InitialMeshOrientation(
    Mesh Orientation Mode=::NewEnumerator2
    Rotation=&RandomRangeVector(Minimum=<0 -0.05 -0.05>, Maximum=<1 0.05 0.05>)
    Use Rotation Vector=true
   )
   &CameraOffset(Camera Offset Amount=5)
  ]
  particleUpdateScript [
   &ParticleState()
   &Timeline(Play Rate=&Lerp_Float(A=0.15, B=0.4, Alpha=$System.AudioReactivity))
   &FN_ListCurve_SkeletalMeshSampler(
    Apply Position=true
    Curve Index:NiagaraFloat=&FloatFromCurve(
     FloatCurve=DataInterface() {propertyValues {Curve {
        defaultValue=1
        keys [{interpMode=RCIM_Cubic, tangentMode=RCTM_User, time=0, value=0}, {interpMode=RCIM_Cubic, tangentMode=RCTM_User, time=1, value=1}]
       }}}
     CurveIndex=$Particles.Timeline.NormalizedTime
    )
    Index_01=19
    Index_02=20
    Index_03=21
    Index_04=21
    Index_05=22
    Index_06=23
    Index_07=24
    Index_08=18
    "Interpolate Mesh Sampling (Update)"=true
    Number of Inputs=6
    Offset Coordinate Space=3
    Offset_01:Vector3f=&AddVectorToPosition(Position=<0 -5 0>, Vector=&RandomRangeVector(Minimum=<-2 -1 -2>, Maximum=<1 1 2>))
    Offset_02:Vector3f=&AddVectorToPosition(
     Position=<0 2.6949 -0.0579>
     Vector=&VectorFromFloat(Value=&RandomRangeFloat(Minimum=-1, Maximum=1))
    )
    Offset_03:Vector3f=&AddVectorToPosition(
     Position=<0 0 8.9572>
     Vector=&VectorFromFloat(Value=&RandomRangeFloat(Minimum=-1.25, Maximum=-1.25))
    )
    Offset_04:Vector3f=&AddVectorToPosition(
     Position=<0 -4.5898 13.454>
     Vector=&VectorFromFloat(Value=&RandomRangeFloat(Minimum=-1.25, Maximum=-1.25))
    )
    Offset_05:Vector3f=&AddVectorToPosition(
     Position=<0 0 6.6452>
     Vector=&VectorFromFloat(Value=&RandomRangeFloat(Minimum=-1.5, Maximum=-1.5))
    )
    Offset_06:Vector3f=&AddVectorToPosition(
     Position=<0 -2.0833 8.8563>
     Vector=&VectorFromFloat(Value=&RandomRangeFloat(Minimum=-1.5, Maximum=-1.5))
    )
    Offset_07:Vector3f=&AddVectorToPosition(
     Position=<-4.4009 -17.1397 -1.8698>
     Vector=&VectorFromFloat(Value=&RandomRangeFloat(Minimum=-1.5, Maximum=-1.5))
    )
    Offset_08:Vector3f=&AddVectorToPosition(Position=<-1.4753 -6.6549 -0.135>, Vector=&RandomRangeVector(Minimum=<-5 -5 -5>, Maximum=<5 2 5>))
    Selection Type=0
    Skeletal Mesh:NiagaraDataInterfaceSkeletalMesh=$System.Skeletal Mesh
   ) {module=FN_ListCurve_SkeletalMeshSampler001}
   &ScaleMeshSize(Scale Factor=&VectorFromFloat(Value=&FloatFromCurve(
      FloatCurve=DataInterface() {propertyValues {Curve {
         defaultValue=0
         keys [
          {arriveTangent=1.098, interpMode=RCIM_Cubic, leaveTangent=1.098, tangentMode=RCTM_User, time=0, value=0.2}
          {arriveTangent=-0.0768, interpMode=RCIM_Cubic, leaveTangent=-0.0768, tangentMode=RCTM_User, time=0.5, value=0.7}
          {arriveTangent=-0.1845, tangentMode=RCTM_User, time=1, value=0}
         ]
        }}}
      CurveIndex=$Particles.Timeline.NormalizedTime
      Scale Curve=&Lerp_Float(
       A=&Lerp_Float(A=$System.DebugScale, B=1, Alpha=$System.Timeline001.NormalizedTime)
       B=&Lerp_Float(A=1.2, B=0.7, Alpha=$System.Timeline.NormalizedTime)
       Alpha=$System.AudioReactivity
      )
     )))
   &ScaleMeshSize(Scale Factor=&VectorFromFloat(Value=&MergeHelper_FloatSwitch(
      0=&Waveform(
       "Waveform [1]"=::NewEnumerator2
       "[1] Frequency"=1
       "[1] Amplitude Min/Max"=<-0.25 0.75>
       "[1] Scale"=$System.AudioLow
       "[1] Bias"=&RandomRangeFloat(Minimum=0.8, Maximum=1.1)
      )
      1=&Waveform(
       "Waveform [1]"=::NewEnumerator2
       "[1] Frequency"=1.5
       "[1] Amplitude Min/Max"=<-0.25 0.75>
       "[1] Scale"=$System.AudioMedium
       "[1] Bias"=&RandomRangeFloat(Minimum=0.8, Maximum=1.1)
      )
      2=&Waveform(
       "Waveform [1]"=::NewEnumerator2
       "[1] Frequency"=2
       "[1] Amplitude Min/Max"=<-0.25 0.75>
       "[1] Scale"=$System.AudioHigh
       "[1] Bias"=&RandomRangeFloat(Minimum=0.8, Maximum=1.1)
      )
      3=?
      Index=$Particles.PartitionParticles.Partition
      Num=::NewEnumerator1
     ))) {module=ScaleMeshSize003}
   &ScaleMeshSize(Scale Factor=&MakeVector(
     X=&Waveform("Waveform [1]"=::NewEnumerator2, "[1] Frequency"=1, "[1] Amplitude Min/Max"=<-0.25 0.25>, "[1] Scale"=$System.AudioLow, "[1] Bias"=1)
     Y=&Waveform("Waveform [1]"=::NewEnumerator2, "[1] Frequency"=2, "[1] Amplitude Min/Max"=<-0.25 0.25>, "[1] Scale"=$System.AudioMedium, "[1] Bias"=1)
     Z=&Waveform("Waveform [1]"=::NewEnumerator2, "[1] Frequency"=3, "[1] Amplitude Min/Max"=<-0.25 0.25>, "[1] Scale"=$System.AudioHigh, "[1] Bias"=1)
    )) {module=ScaleMeshSize001}
   &ScaleColor(
    Scale RGB=&MakeVectorFromLinearColorRGB(Color=&MergeHelper_ColorSwitch(
      0=&LerpLinearColors(
       StartColor=<rgba 1 1 1 1>
       EndColor=&RandomRangeLinearColor(Minimum=<rgba 1 0.4 0.4 1>, Maximum=<rgba 0.4 1 0.4 1>)
       LerpFactor=&Waveform("Waveform [1]"=::NewEnumerator2, "[1] Frequency"=1, "[1] Amplitude Min/Max"=<0 1.2>, "[1] Scale"=$System.AudioLow)
      )
      1=&LerpLinearColors(
       StartColor=<rgba 1 1 1 1>
       EndColor=&RandomRangeLinearColor(Minimum=<rgba 0.4 1 0.4 1>, Maximum=<rgba 0.4 0.4 1 1>)
       LerpFactor=&Waveform("Waveform [1]"=::NewEnumerator2, "[1] Frequency"=1.5, "[1] Amplitude Min/Max"=<0 1.2>, "[1] Scale"=$System.AudioMedium)
      )
      2=&LerpLinearColors(
       StartColor=<rgba 1 1 1 1>
       EndColor=&RandomRangeLinearColor(Minimum=<rgba 0.4 0.4 1 1>, Maximum=<rgba 1 0.4 0.4 1>)
       LerpFactor=&Waveform("Waveform [1]"=::NewEnumerator2, "[1] Frequency"=2, "[1] Amplitude Min/Max"=<0 1.2>, "[1] Scale"=$System.AudioHigh)
      )
      3=?
      Index=$Particles.PartitionParticles.Partition
      Num=::NewEnumerator1
     ))
    ScaleA=true
    ScaleRGB=true
   ) {module=ScaleColor001}
   &SolveForcesAndVelocity(Clamp Velocity=false, Limit Acceleration=false)
   &OffsetPosition(
    Interpolation Curve=DataInterface() {propertyValues {Curve {
       defaultValue=0
       keys [{interpMode=RCIM_Cubic, tangentMode=RCTM_User, time=0, value=0}, {interpMode=RCIM_Cubic, tangentMode=RCTM_User, time=1, value=1}]
      }}}
    Interpolation Scale:NiagaraFloat=$System.AudioReactivity
    Jitter Amount=0.4
    Jitter Interpolation=2
    Jitter Time Offset=0.05
    Offset Mode=1
   )
   &DynamicMaterialParameters(
    Index 0 Param 4=&MergeHelper_FloatSwitch(
     0=&Waveform("Waveform [1]"=::NewEnumerator2, "[1] Frequency"=1, "[1] Amplitude Min/Max"=<0 1>, "[1] Scale"=$System.AudioLow)
     1=&Waveform("Waveform [1]"=::NewEnumerator2, "[1] Frequency"=1.5, "[1] Amplitude Min/Max"=<0 1>, "[1] Scale"=$System.AudioMedium)
     2=&Waveform("Waveform [1]"=::NewEnumerator2, "[1] Frequency"=2, "[1] Amplitude Min/Max"=<0 1>, "[1] Scale"=$System.AudioHigh)
     3=?
     Index=$Particles.PartitionParticles.Partition
     Num=::NewEnumerator1
    )
    Module.Index0Float0Write=true
    Module.Index0Float1Write=true
    Module.Index0Float2Write=true
    Module.Index0Float3Write=true
   )
   /SetVariables(Particles.Position=&Lerp_Vector(
     A=Linked() {linkedVariable {name=Particles.Position, type=NiagaraPosition}}
     B=&AddVectorToPosition(Position=$Particles.Position, Vector=<0 0 -4>)
     Alpha=$System.Timeline.NormalizedTime
    )) {isSetParameters=true, module=SetVariables_#8}
  ]
  rendererClasses [&NiagaraMeshRendererProperties]
  renderers [{rendererClass=&NiagaraMeshRendererProperties, rendererIndex=0}]
 }
 Right_Arm_Voxels001 {
  bEnabled=false
  emitterUpdateScript [
   &EmitterState(
    Inactive Response=::NewEnumerator1
    Life Cycle Mode=::NewEnumerator1
    Loop Behavior=::NewEnumerator1
    Loop Duration Mode=::NewEnumerator1
    UseLoopDelay=false
   )
   &SpawnRate(SpawnRate=&Lerp_Float(A=3, B=8, Alpha=$System.AudioReactivity), Use Spawn Probability=false)
  ]
  particleSpawnScript [
   &PartitionParticles(Number Of Partitions=3, Partition Mode:ENiagara_PartitionMode_v2=::NewEnumerator5)
   &InitializeParticle(
    Color=$System.TODColor
    Lifetime=&Lerp_Float(A=8, B=4, Alpha=$System.AudioReactivity)
    Mesh Scale=&VectorFromFloat(Value=&RandomRangeFloat(Minimum=0.09, Maximum=0.14))
    Mesh Scale Mode=::NewEnumerator4
    Position Mode=::NewEnumerator0
    Ribbon Width=7
    Ribbon Width Mode=::NewEnumerator1
    Sprite Size Mode=::NewEnumerator3
    Write Material Random=true
   )
   &InitialMeshOrientation(
    Mesh Orientation Mode=::NewEnumerator2
    Rotation=&RandomRangeVector(Minimum=<0 -0.05 -0.05>, Maximum=<1 0.05 0.05>)
    Use Rotation Vector=true
   )
   &CameraOffset(Camera Offset Amount=5)
  ]
  particleUpdateScript [
   &ParticleState()
   &Timeline(Play Rate=&Lerp_Float(A=0.15, B=0.4, Alpha=$System.AudioReactivity))
   &FN_ListCurve_SkeletalMeshSampler(
    Apply Position=true
    Curve Index:NiagaraFloat=&FloatFromCurve(
     FloatCurve=DataInterface() {propertyValues {Curve {
        defaultValue=1
        keys [{interpMode=RCIM_Cubic, tangentMode=RCTM_User, time=0, value=0}, {interpMode=RCIM_Cubic, tangentMode=RCTM_User, time=1, value=1}]
       }}}
     CurveIndex=$Particles.Timeline.NormalizedTime
    )
    Index_01=19
    Index_02=20
    Index_03=21
    Index_04=21
    Index_05=22
    Index_06=23
    Index_07=24
    Index_08=18
    "Interpolate Mesh Sampling (Update)"=true
    Number of Inputs=6
    Offset Coordinate Space=3
    Offset_01:Vector3f=&AddVectorToPosition(Position=<0 -5 0>, Vector=&RandomRangeVector(Minimum=<-2 -1 -2>, Maximum=<1 1 2>))
    Offset_02:Vector3f=&AddVectorToPosition(
     Position=<0 2.6949 -0.0579>
     Vector=&VectorFromFloat(Value=&RandomRangeFloat(Minimum=-1, Maximum=1))
    )
    Offset_03:Vector3f=&AddVectorToPosition(
     Position=<0 0 8.9572>
     Vector=&VectorFromFloat(Value=&RandomRangeFloat(Minimum=-1.25, Maximum=-1.25))
    )
    Offset_04:Vector3f=&AddVectorToPosition(
     Position=<0 -4.5898 13.454>
     Vector=&VectorFromFloat(Value=&RandomRangeFloat(Minimum=-1.25, Maximum=-1.25))
    )
    Offset_05:Vector3f=&AddVectorToPosition(
     Position=<0 0 6.6452>
     Vector=&VectorFromFloat(Value=&RandomRangeFloat(Minimum=-1.5, Maximum=-1.5))
    )
    Offset_06:Vector3f=&AddVectorToPosition(
     Position=<0 -2.0833 8.8563>
     Vector=&VectorFromFloat(Value=&RandomRangeFloat(Minimum=-1.5, Maximum=-1.5))
    )
    Offset_07:Vector3f=&AddVectorToPosition(
     Position=<-4.4009 -17.1397 -1.8698>
     Vector=&VectorFromFloat(Value=&RandomRangeFloat(Minimum=-1.5, Maximum=-1.5))
    )
    Offset_08:Vector3f=&AddVectorToPosition(Position=<-1.4753 -6.6549 -0.135>, Vector=&RandomRangeVector(Minimum=<-5 -5 -5>, Maximum=<5 2 5>))
    Selection Type=0
    Skeletal Mesh:NiagaraDataInterfaceSkeletalMesh=$System.Skeletal Mesh
   ) {module=FN_ListCurve_SkeletalMeshSampler001}
   &ScaleMeshSize(Scale Factor=&VectorFromFloat(Value=&FloatFromCurve(
      FloatCurve=DataInterface() {propertyValues {Curve {
         defaultValue=0
         keys [
          {arriveTangent=1.098, interpMode=RCIM_Cubic, leaveTangent=1.098, tangentMode=RCTM_User, time=0, value=0.2}
          {arriveTangent=-0.0768, interpMode=RCIM_Cubic, leaveTangent=-0.0768, tangentMode=RCTM_User, time=0.5, value=0.7}
          {arriveTangent=-0.1845, tangentMode=RCTM_User, time=1, value=0}
         ]
        }}}
      CurveIndex=$Particles.Timeline.NormalizedTime
      Scale Curve=&Lerp_Float(
       A=&Lerp_Float(A=$System.DebugScale, B=1, Alpha=$System.Timeline001.NormalizedTime)
       B=&Lerp_Float(A=1.2, B=0.7, Alpha=$System.Timeline.NormalizedTime)
       Alpha=$System.AudioReactivity
      )
     )))
   &ScaleMeshSize(Scale Factor=&VectorFromFloat(Value=&MergeHelper_FloatSwitch(
      0=&Waveform(
       "Waveform [1]"=::NewEnumerator2
       "[1] Frequency"=1
       "[1] Amplitude Min/Max"=<-0.25 0.75>
       "[1] Scale"=$System.AudioLow
       "[1] Bias"=&RandomRangeFloat(Minimum=0.8, Maximum=1.1)
      )
      1=&Waveform(
       "Waveform [1]"=::NewEnumerator2
       "[1] Frequency"=1.5
       "[1] Amplitude Min/Max"=<-0.25 0.75>
       "[1] Scale"=$System.AudioMedium
       "[1] Bias"=&RandomRangeFloat(Minimum=0.8, Maximum=1.1)
      )
      2=&Waveform(
       "Waveform [1]"=::NewEnumerator2
       "[1] Frequency"=2
       "[1] Amplitude Min/Max"=<-0.25 0.75>
       "[1] Scale"=$System.AudioHigh
       "[1] Bias"=&RandomRangeFloat(Minimum=0.8, Maximum=1.1)
      )
      3=?
      Index=$Particles.PartitionParticles.Partition
      Num=::NewEnumerator1
     ))) {module=ScaleMeshSize003}
   &ScaleMeshSize(Scale Factor=&MakeVector(
     X=&Waveform("Waveform [1]"=::NewEnumerator2, "[1] Frequency"=1, "[1] Amplitude Min/Max"=<-0.25 0.25>, "[1] Scale"=$System.AudioLow, "[1] Bias"=1)
     Y=&Waveform("Waveform [1]"=::NewEnumerator2, "[1] Frequency"=2, "[1] Amplitude Min/Max"=<-0.25 0.25>, "[1] Scale"=$System.AudioMedium, "[1] Bias"=1)
     Z=&Waveform("Waveform [1]"=::NewEnumerator2, "[1] Frequency"=3, "[1] Amplitude Min/Max"=<-0.25 0.25>, "[1] Scale"=$System.AudioHigh, "[1] Bias"=1)
    )) {module=ScaleMeshSize001}
   &ScaleColor(
    Scale RGB=&MakeVectorFromLinearColorRGB(Color=&MergeHelper_ColorSwitch(
      0=&LerpLinearColors(
       StartColor=<rgba 1 1 1 1>
       EndColor=&RandomRangeLinearColor(Minimum=<rgba 1 0.4 0.4 1>, Maximum=<rgba 0.4 1 0.4 1>)
       LerpFactor=&Waveform("Waveform [1]"=::NewEnumerator2, "[1] Frequency"=1, "[1] Amplitude Min/Max"=<0 1.2>, "[1] Scale"=$System.AudioLow)
      )
      1=&LerpLinearColors(
       StartColor=<rgba 1 1 1 1>
       EndColor=&RandomRangeLinearColor(Minimum=<rgba 0.4 1 0.4 1>, Maximum=<rgba 0.4 0.4 1 1>)
       LerpFactor=&Waveform("Waveform [1]"=::NewEnumerator2, "[1] Frequency"=1.5, "[1] Amplitude Min/Max"=<0 1.2>, "[1] Scale"=$System.AudioMedium)
      )
      2=&LerpLinearColors(
       StartColor=<rgba 1 1 1 1>
       EndColor=&RandomRangeLinearColor(Minimum=<rgba 0.4 0.4 1 1>, Maximum=<rgba 1 0.4 0.4 1>)
       LerpFactor=&Waveform("Waveform [1]"=::NewEnumerator2, "[1] Frequency"=2, "[1] Amplitude Min/Max"=<0 1.2>, "[1] Scale"=$System.AudioHigh)
      )
      3=?
      Index=$Particles.PartitionParticles.Partition
      Num=::NewEnumerator1
     ))
    ScaleA=true
    ScaleRGB=true
   ) {module=ScaleColor001}
   &OffsetPosition(
    Interpolation Curve=DataInterface() {propertyValues {Curve {
       defaultValue=0
       keys [{interpMode=RCIM_Cubic, tangentMode=RCTM_User, time=0, value=0}, {interpMode=RCIM_Cubic, tangentMode=RCTM_User, time=1, value=1}]
      }}}
    Interpolation Scale:NiagaraFloat=$System.AudioReactivity
    Jitter Amount=0.4
    Jitter Interpolation=2
    Jitter Time Offset=0.05
    Offset Mode=1
   )
   &DynamicMaterialParameters(
    Index 0 Param 4=&MergeHelper_FloatSwitch(
     0=&Waveform("Waveform [1]"=::NewEnumerator2, "[1] Frequency"=1, "[1] Amplitude Min/Max"=<0 1>, "[1] Scale"=$System.AudioLow)
     1=&Waveform("Waveform [1]"=::NewEnumerator2, "[1] Frequency"=1.5, "[1] Amplitude Min/Max"=<0 1>, "[1] Scale"=$System.AudioMedium)
     2=&Waveform("Waveform [1]"=::NewEnumerator2, "[1] Frequency"=2, "[1] Amplitude Min/Max"=<0 1>, "[1] Scale"=$System.AudioHigh)
     3=?
     Index=$Particles.PartitionParticles.Partition
     Num=::NewEnumerator1
    )
    Module.Index0Float0Write=true
    Module.Index0Float1Write=true
    Module.Index0Float2Write=true
    Module.Index0Float3Write=true
   )
   /SetVariables(Particles.Position=&Lerp_Vector(
     A=Linked() {linkedVariable {name=Particles.Position, type=NiagaraPosition}}
     B=&AddVectorToPosition(Position=$Particles.Position, Vector=<0 0 -4>)
     Alpha=$System.Timeline.NormalizedTime
    )) {isSetParameters=true, module=SetVariables_#8}
  ]
  rendererClasses [&NiagaraMeshRendererProperties]
  renderers [{rendererClass=&NiagaraMeshRendererProperties, rendererIndex=0}]
 }
 Right_Leg_Voxels {
  emitterUpdateScript [
   &EmitterState(
    Inactive Response=::NewEnumerator1
    Life Cycle Mode=::NewEnumerator1
    Loop Behavior=::NewEnumerator1
    Loop Duration Mode=::NewEnumerator1
    UseLoopDelay=false
   )
   &SpawnRate(SpawnRate=&Lerp_Float(A=3, B=6, Alpha=$System.AudioReactivity), Use Spawn Probability=false)
  ]
  particleSpawnScript [
   &PartitionParticles(Number Of Partitions=3, Partition Mode:ENiagara_PartitionMode_v2=::NewEnumerator5)
   &InitializeParticle(
    Color=$System.TODColor
    Lifetime=&Lerp_Float(A=6, B=3, Alpha=$System.AudioReactivity)
    Mesh Scale=&VectorFromFloat(Value=&RandomRangeFloat(Minimum=0.09, Maximum=0.14))
    Mesh Scale Mode=::NewEnumerator4
    Position Mode=::NewEnumerator0
    Ribbon Width=7
    Ribbon Width Mode=::NewEnumerator1
    Sprite Size Mode=::NewEnumerator3
    Write Material Random=true
   )
   &InitialMeshOrientation(
    Mesh Orientation Mode=::NewEnumerator2
    Rotation=&RandomRangeVector(Minimum=<0 -0.05 -0.05>, Maximum=<1 0.05 0.05>)
    Use Rotation Vector=true
   )
  ]
  particleUpdateScript [
   &ParticleState()
   &Timeline(Play Rate=&Lerp_Float(A=0.2, B=0.4, Alpha=$System.AudioReactivity))
   &FN_ListCurve_SkeletalMeshSampler(
    Apply Position=true
    Curve Index:NiagaraFloat=&FloatFromCurve(FloatCurve=DataInterface() {propertyValues {Curve {
        defaultValue=1
        keys [
         {arriveTangent=0.9614, interpMode=RCIM_Cubic, leaveTangent=0.9614, tangentMode=RCTM_User, time=0, value=0}
         {interpMode=RCIM_Cubic, tangentMode=RCTM_User, time=1, value=1}
        ]
       }}})
    Index_01=10
    Index_02=10
    Index_03=11
    Index_04=11
    Index_05=11
    Index_06=12
    Index_07=12
    Number of Inputs=5
    Offset Coordinate Space=3
    Offset_01:Vector3f=&AddVectorToPosition(Position=<2.4966 5.0004 0>, Vector=&RandomRangeVector(Minimum=<-2 -2 -2>, Maximum=<2 2 2>))
    Offset_02:Vector3f=&AddVectorToPosition(
     Position=<-2.3226 8.024 1.9512>
     Vector=&VectorFromFloat(Value=&RandomRangeFloat(Minimum=-1, Maximum=1))
    )
    Offset_03:Vector3f=&AddVectorToPosition(
     Position=<4.1941 9.9182 -4.4708>
     Vector=&VectorFromFloat(Value=&RandomRangeFloat(Minimum=-1.5, Maximum=1.5))
    )
    Offset_04:Vector3f=&AddVectorToPosition(
     Position=<1.9425 8.3694 -0.3202>
     Vector=&VectorFromFloat(Value=&RandomRangeFloat(Minimum=-1, Maximum=1))
    )
    Offset_05:Vector3f=&AddVectorToPosition(
     Position=<1.2524 -0.7642 3.055>
     Vector=&VectorFromFloat(Value=&RandomRangeFloat(Minimum=-2, Maximum=2))
    )
    Offset_06:Vector3f=&AddVectorToPosition(
     Position=<-7.0519 4.3101 0.9751>
     Vector=&VectorFromFloat(Value=&RandomRangeFloat(Minimum=-2, Maximum=2))
    )
    Offset_07:Vector3f=&AddVectorToPosition(Position=<-0.1625 0.274 4.5005>, Vector=&RandomRangeVector(Minimum=<-5 0 -5>, Maximum=<5 0 5>))
    Selection Type=0
    Skeletal Mesh:NiagaraDataInterfaceSkeletalMesh=$System.Skeletal Mesh
   ) {module=FN_ListCurve_SkeletalMeshSampler001}
   &ScaleMeshSize(Scale Factor=&VectorFromFloat(Value=&FloatFromCurve(
      FloatCurve=DataInterface() {propertyValues {Curve {
         defaultValue=0
         keys [
          {arriveTangent=1.098, interpMode=RCIM_Cubic, leaveTangent=1.098, tangentMode=RCTM_User, time=0, value=0.2}
          {arriveTangent=-0.245, interpMode=RCIM_Cubic, leaveTangent=-0.245, tangentMode=RCTM_User, time=0.5, value=0.7}
          {arriveTangent=0.0898, tangentMode=RCTM_User, time=1, value=0}
         ]
        }}}
      CurveIndex=$Particles.Timeline.NormalizedTime
      Scale Curve=&Lerp_Float(A=$System.DebugScale, B=1.2, Alpha=$System.Timeline001.NormalizedTime)
     )))
   &ScaleMeshSize(Scale Factor=&VectorFromFloat(Value=&MergeHelper_FloatSwitch(
      0=&Waveform(
       "Waveform [1]"=::NewEnumerator2
       "[1] Frequency"=1
       "[1] Amplitude Min/Max"=<-0.25 0.75>
       "[1] Scale"=$System.AudioLow
       "[1] Bias"=&RandomRangeFloat(Minimum=0.8, Maximum=1.1)
      )
      1=&Waveform(
       "Waveform [1]"=::NewEnumerator2
       "[1] Frequency"=1.5
       "[1] Amplitude Min/Max"=<-0.25 0.75>
       "[1] Scale"=$System.AudioMedium
       "[1] Bias"=&RandomRangeFloat(Minimum=0.8, Maximum=1.1)
      )
      2=&Waveform(
       "Waveform [1]"=::NewEnumerator2
       "[1] Frequency"=2
       "[1] Amplitude Min/Max"=<-0.25 0.75>
       "[1] Scale"=$System.AudioHigh
       "[1] Bias"=&RandomRangeFloat(Minimum=0.8, Maximum=1.1)
      )
      3=?
      Index=$Particles.PartitionParticles.Partition
      Num=::NewEnumerator1
     ))) {module=ScaleMeshSize003}
   &ScaleMeshSize(Scale Factor=&MakeVector(
     X=&Waveform("Waveform [1]"=::NewEnumerator2, "[1] Frequency"=1, "[1] Amplitude Min/Max"=<-0.25 0.25>, "[1] Scale"=$System.AudioLow, "[1] Bias"=1)
     Y=&Waveform("Waveform [1]"=::NewEnumerator2, "[1] Frequency"=2, "[1] Amplitude Min/Max"=<-0.25 0.25>, "[1] Scale"=$System.AudioMedium, "[1] Bias"=1)
     Z=&Waveform("Waveform [1]"=::NewEnumerator2, "[1] Frequency"=3, "[1] Amplitude Min/Max"=<-0.25 0.25>, "[1] Scale"=$System.AudioHigh, "[1] Bias"=1)
    )) {module=ScaleMeshSize001}
   &ScaleColor(
    Scale RGB=&MakeVectorFromLinearColorRGB(Color=&MergeHelper_ColorSwitch(
      0=&LerpLinearColors(
       StartColor=<rgba 1 1 1 1>
       EndColor=&RandomRangeLinearColor(Minimum=<rgba 1 0.4 0.4 1>, Maximum=<rgba 0.4 1 0.4 1>)
       LerpFactor=&Waveform("Waveform [1]"=::NewEnumerator2, "[1] Frequency"=1, "[1] Amplitude Min/Max"=<0 1.2>, "[1] Scale"=$System.AudioLow)
      )
      1=&LerpLinearColors(
       StartColor=<rgba 1 1 1 1>
       EndColor=&RandomRangeLinearColor(Minimum=<rgba 0.4 1 0.4 1>, Maximum=<rgba 0.4 0.4 1 1>)
       LerpFactor=&Waveform("Waveform [1]"=::NewEnumerator2, "[1] Frequency"=1.5, "[1] Amplitude Min/Max"=<0 1.2>, "[1] Scale"=$System.AudioMedium)
      )
      2=&LerpLinearColors(
       StartColor=<rgba 1 1 1 1>
       EndColor=&RandomRangeLinearColor(Minimum=<rgba 0.4 0.4 1 1>, Maximum=<rgba 1 0.4 0.4 1>)
       LerpFactor=&Waveform("Waveform [1]"=::NewEnumerator2, "[1] Frequency"=2, "[1] Amplitude Min/Max"=<0 1.2>, "[1] Scale"=$System.AudioHigh)
      )
      3=?
      Index=$Particles.PartitionParticles.Partition
      Num=::NewEnumerator1
     ))
    ScaleA=true
    ScaleRGB=true
   ) {module=ScaleColor001}
   &SolveForcesAndVelocity(Clamp Velocity=false, Limit Acceleration=false)
   &OffsetPosition(Interpolation Scale:NiagaraFloat=$System.AudioReactivity, Jitter Amount=0.15, Jitter Interpolation=1, Jitter Time Offset=0.05, Offset Mode=1)
   &DynamicMaterialParameters(
    Index 0 Param 4=&MergeHelper_FloatSwitch(
     0=&Waveform("Waveform [1]"=::NewEnumerator8, "[1] Frequency"=1, "[1] Amplitude Min/Max"=<0 1>, "[1] Scale"=$System.AudioLow)
     1=&Waveform("Waveform [1]"=::NewEnumerator8, "[1] Frequency"=1.5, "[1] Amplitude Min/Max"=<0 1>, "[1] Scale"=$System.AudioMedium)
     2=&Waveform("Waveform [1]"=::NewEnumerator8, "[1] Frequency"=2, "[1] Amplitude Min/Max"=<0 1>, "[1] Scale"=$System.AudioHigh)
     3=?
     Index=$Particles.PartitionParticles.Partition
     Num=::NewEnumerator1
    )
    Module.Index0Float0Write=true
    Module.Index0Float1Write=true
    Module.Index0Float2Write=true
    Module.Index0Float3Write=true
   )
  ]
  rendererClasses [&NiagaraMeshRendererProperties]
  renderers [{rendererClass=&NiagaraMeshRendererProperties, rendererIndex=0}]
 }
}
summary {
 systemName=NS_Character_AnvilAlarmMoth_Idle_DebugBackup
 emitters [
  {emitterName=Heart, rendererClasses [&NiagaraMeshRendererProperties]}
  {bEnabled=false, emitterName=FN_Debug001, rendererClasses [&NiagaraSpriteRendererProperties]}
  {bEnabled=false, emitterName=FN_Debug002, rendererClasses [&NiagaraSpriteRendererProperties]}
  {
   bEnabled=false
   emitterName=LeftLeg
   rendererClasses [
    &NiagaraRibbonRendererProperties
    &NiagaraSpriteRendererProperties
    &NiagaraMeshRendererProperties
   ]
  }
  {
   bEnabled=false
   emitterName=RightLeg
   rendererClasses [
    &NiagaraRibbonRendererProperties
    &NiagaraSpriteRendererProperties
    &NiagaraMeshRendererProperties
   ]
  }
  {emitterName=Left_Leg_Voxels, rendererClasses [&NiagaraMeshRendererProperties]}
  {emitterName=Right_Leg_Voxels, rendererClasses [&NiagaraMeshRendererProperties]}
  {emitterName=Left_Arm_Voxels, rendererClasses [&NiagaraMeshRendererProperties]}
  {
   bEnabled=false
   emitterName=LeftArm001
   rendererClasses [
    &NiagaraRibbonRendererProperties
    &NiagaraSpriteRendererProperties
    &NiagaraMeshRendererProperties
   ]
  }
  {
   bEnabled=false
   emitterName=RightArm
   rendererClasses [
    &NiagaraRibbonRendererProperties
    &NiagaraSpriteRendererProperties
    &NiagaraMeshRendererProperties
   ]
  }
  {emitterName=Right_Arm_Voxels, rendererClasses [&NiagaraMeshRendererProperties]}
  {bEnabled=false, emitterName=Left_Arm_Voxels001, rendererClasses [&NiagaraMeshRendererProperties]}
  {bEnabled=false, emitterName=Right_Arm_Voxels001, rendererClasses [&NiagaraMeshRendererProperties]}
 ]
 userVariables [
  User.AudioType {type=NiagaraFloat, defaultValue=1, description=""}
  User.bIsADS {type=NiagaraBool}
  User.bIsFrontEnd {type=NiagaraBool}
  User.bIsFrontEndPreview {type=NiagaraBool}
  User.DebugReactivity {type=NiagaraBool}
  User.EnableFrontendFakeAudioLoop {type=NiagaraFloat, defaultValue=0, description=""}
  User.IsSkydiving {type=NiagaraBool}
 ]
}
