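# Physics asset for the chain on the AnvilAlarmMoth male medium player body cosmetic, modelling it as capsule bodies.
# Front and back five-link strands start at base bones, with limited swing and locked twist between links.
# An angularly free constraint joins the two end links, closing the strands into a hanging loop.
# Capsules on both thighs and the right calf add leg volumes, including a second, oversized one on the right thigh.
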PhysicsAsset /BRCosmetics/Characters/Player/Male/Medium/Bodies/M_MED_AnvilAlarmMoth/Meshes/M_MED_AnvilAlarmMoth_Chain_Physics
bodies {
 base_chain_back [{shapeType=Capsule, length=0.505, radius=0.505, shapeName=base_chain_back_capsule, rotation=<rot 0 -90 0>}]
 base_chain_front [{shapeType=Capsule, length=0.505, radius=0.505, shapeName=base_chain_front_capsule, rotation=<rot 0 -90 0>}]
 base_offset_chain_back [{
   shapeType=Capsule
   length=0.987
   radius=0.987
   shapeName=base_offset_chain_back_capsule
   center=<0.7582 -0.3318 0.0134>
   rotation=<rot 0 67.3405 -89.0662>
  }]
 calf_r [{
   shapeType=Capsule
   length=34.392
   radius=6.608
   shapeName=thigh_r_capsule
   center=<20.0543 1.5 -1.3732>
   rotation=<rot 88.5654 179.9995 179.9996>
  }]
 dyn_chain_back_01 [{shapeType=Capsule, length=3, radius=2, shapeName=dyn_chain_back_01_capsule, center=<2 0 0.5>, rotation=<rot 90 0 0>}]
 dyn_chain_back_02 [{shapeType=Capsule, length=3, radius=2, shapeName=dyn_chain_back_02_capsule, center=<1 0 0.5>, rotation=<rot 90 0 0>}]
 dyn_chain_back_03 [{shapeType=Capsule, length=3, radius=2, shapeName=dyn_chain_back_03_capsule, center=<2 0 0.5>, rotation=<rot 90 0 0>}]
 dyn_chain_back_04 [{shapeType=Capsule, length=3, radius=2, shapeName=dyn_chain_back_04_capsule, center=<2 0 0.5>, rotation=<rot 90 0 0>}]
 dyn_chain_back_05 [{
   shapeType=Capsule
   length=3
   radius=2
   shapeName=dyn_chain_back_04_capsule
   center=<2.6871 -0.671 0.5>
   rotation=<rot 70.5541 -27.2732 3.0037>
  }]
 dyn_chain_front_01 [{shapeType=Capsule, length=3, radius=1.5, shapeName=dyn_chain_front_01_capsule, center=<3 -1 -1>, rotation=<rot 90 0 0>}]
 dyn_chain_front_02 [{shapeType=Capsule, length=3, radius=1.5, shapeName=dyn_chain_front_02_capsule, center=<2 -1 -1>, rotation=<rot 90 0 0>}]
 dyn_chain_front_03 [{shapeType=Capsule, length=3, radius=2, shapeName=dyn_chain_front_03_capsule, center=<2 0 -0.5>, rotation=<rot 90 0 0>}]
 dyn_chain_front_04 [{shapeType=Capsule, length=3, radius=2, shapeName=dyn_chain_front_04_capsule, center=<2 0 -0.5>, rotation=<rot 90 0 0>}]
 dyn_chain_front_05 [{
   shapeType=Capsule
   length=3
   radius=2
   shapeName=dyn_chain_front_04_capsule
   center=<2.8254 0 -0.0249>
   rotation=<rot 69.8264 179.9951 -178.4641>
  }]
 thigh_l [{
   shapeType=Capsule
   length=31.7677
   radius=7.6762
   shapeName=thigh_r_capsule
   center=<-21.0463 0.1696 2.531>
   rotation=<rot 0 -91.5117 -88.1966>
  }]
 thigh_r [
  {
   shapeType=Capsule
   length=31.1237
   radius=7.9578
   shapeName=thigh_r_capsule
   center=<21.0661 0.581 -2.531>
   rotation=<rot 0 -91.5117 -89.5958>
  }
  {shapeType=Capsule, length=48.7557, radius=15, shapeName=thigh_r_capsule, center=<21.0336 0 -16.6242>, rotation=<rot -88.0073 -0.0016 0.001>}
 ]
}
constraints [
 {
  bone1Name=dyn_chain_back_01
  bone2Name=base_chain_back
  swing1LimitDegrees=30
  swing2LimitDegrees=20
  twistLimitDegrees=45
  twistMotion=Locked
 }
 {
  bone1Name=dyn_chain_back_02
  bone2Name=dyn_chain_back_01
  swing1LimitDegrees=20
  swing2LimitDegrees=20
  twistLimitDegrees=45
  twistMotion=Locked
 }
 {
  bone1Name=dyn_chain_back_03
  bone2Name=dyn_chain_back_02
  swing1LimitDegrees=20
  swing2LimitDegrees=20
  twistLimitDegrees=45
  twistMotion=Locked
 }
 {
  bone1Name=dyn_chain_back_04
  bone2Name=dyn_chain_back_03
  swing1LimitDegrees=20
  swing2LimitDegrees=20
  twistLimitDegrees=45
  twistMotion=Locked
 }
 {
  bone1Name=dyn_chain_back_05
  bone2Name=dyn_chain_back_04
  swing1LimitDegrees=20
  swing2LimitDegrees=20
  twistLimitDegrees=45
  twistMotion=Locked
 }
 {
  bone1Name=dyn_chain_front_01
  bone2Name=base_chain_front
  swing1LimitDegrees=30
  swing2LimitDegrees=20
  twistLimitDegrees=45
  twistMotion=Locked
 }
 {
  bone1Name=dyn_chain_front_02
  bone2Name=dyn_chain_front_01
  swing1LimitDegrees=20
  swing2LimitDegrees=20
  twistLimitDegrees=45
  twistMotion=Locked
 }
 {
  bone1Name=dyn_chain_front_03
  bone2Name=dyn_chain_front_02
  swing1LimitDegrees=20
  swing2LimitDegrees=20
  twistLimitDegrees=45
  twistMotion=Locked
 }
 {
  bone1Name=dyn_chain_front_04
  bone2Name=dyn_chain_front_03
  swing1LimitDegrees=20
  swing2LimitDegrees=20
  twistLimitDegrees=45
  twistMotion=Locked
 }
 {
  bone1Name=dyn_chain_front_05
  bone2Name=dyn_chain_front_04
  swing1LimitDegrees=20
  swing2LimitDegrees=20
  twistLimitDegrees=45
  twistMotion=Locked
 }
 {
  bone1Name=dyn_chain_front_05
  bone2Name=dyn_chain_back_05
  swing1LimitDegrees=45
  swing1Motion=Free
  swing2LimitDegrees=45
  swing2Motion=Free
  twistLimitDegrees=45
  twistMotion=Free
 }
 {
  bone1Name=calf_r
  bone2Name=thigh_r
  swing1LimitDegrees=45
  swing1Motion=Free
  swing2LimitDegrees=45
  swing2Motion=Free
  twistLimitDegrees=45
  twistMotion=Free
 }
 {
  bone1Name=base_chain_back
  bone2Name=base_offset_chain_back
  swing1LimitDegrees=45
  swing1Motion=Free
  swing2LimitDegrees=45
  swing2Motion=Free
  twistLimitDegrees=45
  twistMotion=Free
 }
]
