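# Physics asset for the lower jacket of the AnvilAlarmMoth medium male player body, simulating jacket flaps, belt and lighter.
# Seven two-segment capsule chains hang from MSTR root bodies constrained to spine_01, with tight swing and twist limits along each chain.
# Mostly free cross-constraints link some neighbouring lower segments.
# Pelvis, spine and thigh capsules form the body volume, with belt and lighter attached to the pelvis by free joints.
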PhysicsAsset /BRCosmetics/Characters/Player/Male/Medium/Bodies/M_MED_AnvilAlarmMoth/Meshes/M_MED_AnvilAlarmMoth_LowerJacket_Physics
bodies {
 dyn_belt_01 [{
   shapeType=Capsule
   length=4.8922
   radius=2.5733
   shapeName=dyn_belt_02_capsule
   center=<2.0078 -0.0574 0.111>
   rotation=<rot 0 92.2354 -88.8896>
  }]
 dyn_belt_02 [{
   shapeType=Capsule
   length=4.8922
   radius=2.5733
   shapeName=dyn_belt_02_capsule
   center=<2.0078 -0.0574 0.111>
   rotation=<rot 0 92.2354 -88.8896>
  }]
 dyn_jacket_bk_01 [{shapeType=Capsule, length=14.317, radius=7.6129, shapeName=dyn_jacket_l_bk_01_capsule, center=<6 0 5.0909>, rotation=<rot 0 0 90>}]
 dyn_jacket_bk_02 [{shapeType=Capsule, length=14.5266, radius=7.4033, shapeName=dyn_jacket_l_bk_02_capsule, center=<6 0 4.1579>, rotation=<rot 0 0 90>}]
 dyn_jacket_bk_MSTR [{shapeType=Capsule, length=1, radius=1}]
 dyn_jacket_l_01 [{shapeType=Capsule, length=14.3188, radius=6.3717, shapeName=dyn_jacket_l_01_capsule, center=<6 0 7>, rotation=<rot 0 0 90>}]
 dyn_jacket_l_02 [{shapeType=Capsule, length=15.585, radius=7.4516, shapeName=dyn_jacket_l_02_capsule, center=<6 0 7>, rotation=<rot 0 0 90>}]
 dyn_jacket_l_MSTR [{shapeType=Capsule, length=1, radius=1}]
 dyn_jacket_l_bk_01 [{
   shapeType=Capsule
   length=17.2673
   radius=6.3241
   shapeName=dyn_jacket_l_bk_01_capsule
   center=<6 -0.0401 6.5418>
   rotation=<rot -0.0001 0.0001 95>
  }]
 dyn_jacket_l_bk_02 [{
   shapeType=Capsule
   length=15.9251
   radius=7.6663
   shapeName=dyn_jacket_l_bk_02_capsule
   center=<6 -0.0401 6.5418>
   rotation=<rot -0.0001 0.0001 95>
  }]
 dyn_jacket_l_bk_MSTR [{shapeType=Capsule, length=1, radius=1}]
 dyn_jacket_l_fr_01 [{shapeType=Capsule, length=4.1912, radius=7.4601, shapeName=dyn_jacket_l_fr_01_capsule, center=<6 0 7>, rotation=<rot 90 180 180>}]
 dyn_jacket_l_fr_02 [{shapeType=Capsule, length=8.4398, radius=7.4601, shapeName=dyn_jacket_l_fr_02_capsule, center=<3.4299 -1 8>, rotation=<rot 0 0 105>}]
 dyn_jacket_l_fr_MSTR [{shapeType=Capsule, length=1, radius=1}]
 dyn_jacket_r_01 [{shapeType=Capsule, length=15.8142, radius=7.4704, shapeName=dyn_jacket_l_01_capsule, center=<6 0 6.6101>, rotation=<rot 0 0 90>}]
 dyn_jacket_r_02 [{shapeType=Capsule, length=16.2941, radius=6.9905, shapeName=dyn_jacket_l_02_capsule, center=<6 0 5.9329>, rotation=<rot 0 0 90>}]
 dyn_jacket_r_MSTR [{shapeType=Capsule, length=1, radius=1}]
 dyn_jacket_r_bk_01 [{
   shapeType=Capsule
   length=12.5847
   radius=6.4379
   shapeName=dyn_jacket_l_bk_01_capsule
   center=<6 0.022 6.7489>
   rotation=<rot -0.0001 -0.0001 85>
  }]
 dyn_jacket_r_bk_02 [{
   shapeType=Capsule
   length=12.6689
   radius=6.3537
   shapeName=dyn_jacket_l_bk_02_capsule
   center=<6 0 6.7479>
   rotation=<rot -0.0002 -0.0001 90>
  }]
 dyn_jacket_r_bk_MSTR [{shapeType=Capsule, length=1, radius=1}]
 dyn_jacket_r_fr_01 [{shapeType=Capsule, length=4.1912, radius=7.4601, shapeName=dyn_jacket_l_fr_01_capsule, center=<6 0 7>, rotation=<rot 90 180 180>}]
 dyn_jacket_r_fr_02 [{shapeType=Capsule, length=7.5151, radius=7.4601, shapeName=dyn_jacket_l_fr_02_capsule, center=<6 2 8>, rotation=<rot 0.0001 -0.0002 80>}]
 dyn_jacket_r_fr_MSTR [{shapeType=Capsule, length=1, radius=1}]
 dyn_lighter [{
   shapeType=Capsule
   length=6.2365
   radius=4.9326
   shapeName=dyn_lighter_04_capsule
   center=<3.1956 -0.4551 -3.2042>
   rotation=<rot 90 23.2557 0>
  }]
 pelvis [{shapeType=Capsule, length=8.4244, radius=9.6849, center=<-0.9769 -0.7557 0.2377>}]
 spine_01 [{shapeType=Capsule, length=7.976, radius=9.7034, center=<-1.072 -0.8263 0>}]
 thigh_l [{shapeType=Capsule, length=30.9367, radius=9.0753, center=<-14.0559 -0.4384 3.6508>, rotation=<rot 87.788 -32.1044 146.5476>}]
 thigh_twist_01_r [{shapeType=Capsule, length=24.0514, radius=8.6805, center=<-7.9504 0.3594 -2.8213>, rotation=<rot 89.6563 163.6498 161.8731>}]
}
constraints [
 {bone1Name=dyn_jacket_r_bk_01, bone2Name=dyn_jacket_r_bk_MSTR, swing1LimitDegrees=5, swing2LimitDegrees=20, twistLimitDegrees=2}
 {bone1Name=dyn_jacket_r_bk_02, bone2Name=dyn_jacket_r_bk_01, swing1LimitDegrees=5, swing2LimitDegrees=20, twistLimitDegrees=2}
 {bone1Name=dyn_jacket_l_bk_01, bone2Name=dyn_jacket_l_bk_MSTR, swing1LimitDegrees=5, swing2LimitDegrees=20, twistLimitDegrees=2}
 {bone1Name=dyn_jacket_l_bk_02, bone2Name=dyn_jacket_l_bk_01, swing1LimitDegrees=5, swing2LimitDegrees=20, twistLimitDegrees=2}
 {bone1Name=dyn_jacket_l_fr_01, bone2Name=dyn_jacket_l_fr_MSTR, swing1LimitDegrees=5, swing2LimitDegrees=30, twistLimitDegrees=5}
 {bone1Name=dyn_jacket_l_fr_02, bone2Name=dyn_jacket_l_fr_01, swing1LimitDegrees=5, swing2LimitDegrees=60, twistLimitDegrees=5}
 {bone1Name=dyn_jacket_r_fr_01, bone2Name=dyn_jacket_r_fr_MSTR, swing1LimitDegrees=5, swing2LimitDegrees=30, twistLimitDegrees=5}
 {bone1Name=dyn_jacket_r_fr_02, bone2Name=dyn_jacket_r_fr_01, swing1LimitDegrees=5, swing2LimitDegrees=60, twistLimitDegrees=5}
 {bone1Name=dyn_jacket_l_02, bone2Name=dyn_jacket_l_01, swing1LimitDegrees=2, swing2LimitDegrees=60, twistLimitDegrees=2}
 {bone1Name=dyn_jacket_l_01, bone2Name=dyn_jacket_l_MSTR, swing1LimitDegrees=2, swing2LimitDegrees=30, twistLimitDegrees=2}
 {bone1Name=dyn_jacket_r_01, bone2Name=dyn_jacket_r_MSTR, swing1LimitDegrees=5, swing2LimitDegrees=30, twistLimitDegrees=2}
 {bone1Name=dyn_jacket_r_02, bone2Name=dyn_jacket_r_01, swing1LimitDegrees=5, swing2LimitDegrees=40, twistLimitDegrees=2}
 {
  bone1Name=dyn_jacket_r_bk_MSTR
  bone2Name=spine_01
  swing1LimitDegrees=45
  swing1Motion=Free
  swing2LimitDegrees=45
  swing2Motion=Free
  twistLimitDegrees=45
  twistMotion=Free
 }
 {
  bone1Name=dyn_jacket_l_bk_MSTR
  bone2Name=spine_01
  swing1LimitDegrees=45
  swing1Motion=Free
  swing2LimitDegrees=45
  swing2Motion=Free
  twistLimitDegrees=45
  twistMotion=Free
 }
 {
  bone1Name=dyn_jacket_bk_MSTR
  bone2Name=spine_01
  swing1LimitDegrees=45
  swing1Motion=Free
  swing2LimitDegrees=45
  swing2Motion=Free
  twistLimitDegrees=45
  twistMotion=Free
 }
 {
  bone1Name=dyn_jacket_r_MSTR
  bone2Name=spine_01
  swing1LimitDegrees=45
  swing1Motion=Free
  swing2LimitDegrees=45
  swing2Motion=Free
  twistLimitDegrees=45
  twistMotion=Free
 }
 {
  bone1Name=dyn_jacket_l_MSTR
  bone2Name=spine_01
  swing1LimitDegrees=45
  swing1Motion=Free
  swing2LimitDegrees=45
  swing2Motion=Free
  twistLimitDegrees=45
  twistMotion=Free
 }
 {
  bone1Name=dyn_jacket_l_fr_MSTR
  bone2Name=spine_01
  swing1LimitDegrees=45
  swing1Motion=Free
  swing2LimitDegrees=45
  swing2Motion=Free
  twistLimitDegrees=45
  twistMotion=Free
 }
 {
  bone1Name=dyn_jacket_r_fr_MSTR
  bone2Name=spine_01
  swing1LimitDegrees=45
  swing1Motion=Free
  swing2LimitDegrees=45
  swing2Motion=Free
  twistLimitDegrees=45
  twistMotion=Free
 }
 {bone1Name=dyn_jacket_bk_01, bone2Name=dyn_jacket_bk_MSTR, swing1LimitDegrees=5, swing2LimitDegrees=20, twistLimitDegrees=2}
 {bone1Name=dyn_jacket_bk_02, bone2Name=dyn_jacket_bk_01, swing1LimitDegrees=5, swing2LimitDegrees=20, twistLimitDegrees=2}
 {
  bone1Name=spine_01
  bone2Name=pelvis
  swing1LimitDegrees=45
  swing1Motion=Free
  swing2LimitDegrees=45
  swing2Motion=Free
  twistLimitDegrees=45
  twistMotion=Free
 }
 {
  bone1Name=thigh_twist_01_r
  bone2Name=pelvis
  swing1LimitDegrees=45
  swing1Motion=Free
  swing2LimitDegrees=45
  swing2Motion=Free
  twistLimitDegrees=45
  twistMotion=Free
 }
 {
  bone1Name=thigh_l
  bone2Name=pelvis
  swing1LimitDegrees=45
  swing1Motion=Free
  swing2LimitDegrees=45
  swing2Motion=Free
  twistLimitDegrees=45
  twistMotion=Free
 }
 {
  bone1Name=dyn_jacket_l_bk_02
  bone2Name=dyn_jacket_bk_02
  swing1LimitDegrees=45
  swing1Motion=Free
  swing2LimitDegrees=45
  swing2Motion=Free
  twistLimitDegrees=45
  twistMotion=Free
 }
 {
  bone1Name=dyn_jacket_r_fr_02
  bone2Name=dyn_jacket_r_02
  swing1LimitDegrees=45
  swing1Motion=Free
  swing2LimitDegrees=10
  twistLimitDegrees=45
  twistMotion=Free
 }
 {
  bone1Name=dyn_jacket_r_02
  bone2Name=dyn_jacket_r_bk_02
  swing1LimitDegrees=45
  swing1Motion=Free
  swing2LimitDegrees=45
  swing2Motion=Free
  twistLimitDegrees=45
  twistMotion=Free
 }
 {
  bone1Name=dyn_jacket_l_02
  bone2Name=dyn_jacket_l_bk_02
  swing1LimitDegrees=45
  swing1Motion=Free
  swing2LimitDegrees=45
  swing2Motion=Free
  twistLimitDegrees=45
  twistMotion=Free
 }
 {
  bone1Name=dyn_lighter
  bone2Name=pelvis
  swing1LimitDegrees=45
  swing1Motion=Free
  swing2LimitDegrees=45
  swing2Motion=Free
  twistLimitDegrees=45
  twistMotion=Free
 }
 {
  bone1Name=dyn_belt_01
  bone2Name=pelvis
  swing1LimitDegrees=45
  swing1Motion=Free
  swing2LimitDegrees=45
  swing2Motion=Free
  twistLimitDegrees=45
  twistMotion=Free
 }
 {
  bone1Name=dyn_belt_02
  bone2Name=dyn_belt_01
  swing1LimitDegrees=45
  swing1Motion=Free
  swing2LimitDegrees=45
  swing2Motion=Free
  twistLimitDegrees=45
  twistMotion=Free
 }
]
