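# Physics asset for the upper-body mesh of the AnvilAlarmMoth male medium cosmetic, covering jiggle parts at the neck, necklaces and jacket ornaments.
# Capsules on spine, chest, neck and head form the body volume, with tiny base_ anchor bodies mostly free-jointed to the spine.
# The dyn_ bodies hang from those anchors on limited swing: tight neck pieces, two three-link necklace chains, and twist-locked ornaments.
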PhysicsAsset /BRCosmetics/Characters/Player/Male/Medium/Bodies/M_MED_AnvilAlarmMoth/Meshes/M_MED_AnvilAlarmMoth_Up_Physics
bodies {
 base_jacket_ornament_l [{shapeType=Capsule, length=0.505, radius=0.505, shapeName=base_jacket_ornament_l_capsule, rotation=<rot 0 -90 0>}]
 base_jacket_ornament_r [{shapeType=Capsule, length=0.505, radius=0.505, shapeName=base_jacket_ornament_r_capsule, rotation=<rot 0 -90 0>}]
 base_neck_back [{shapeType=Capsule, length=0.505, radius=0.505, shapeName=base_neck_back_capsule, rotation=<rot 0 -90 0>}]
 base_neck_front_l [{shapeType=Capsule, length=0.505, radius=0.505, shapeName=base_neck_front_l_capsule, rotation=<rot 0 -90 0>}]
 base_neck_front_r [{shapeType=Capsule, length=0.505, radius=0.505, shapeName=base_neck_front_r_capsule, rotation=<rot 0 -90 0>}]
 base_neck_mid_l [{shapeType=Capsule, length=0.505, radius=0.505, shapeName=base_neck_mid_l_capsule, rotation=<rot 0 -90 0>}]
 base_neck_mid_r [{shapeType=Capsule, length=0.505, radius=0.505, shapeName=base_neck_mid_r_capsule, rotation=<rot 0 -90 0>}]
 base_neck_midback_r [{shapeType=Capsule, length=0.505, radius=0.505, shapeName=base_neck_midback_r_capsule, rotation=<rot 0 -90 0>}]
 base_necklace_in [{shapeType=Capsule, length=0.2, radius=0.2, shapeName=base_necklace_in_capsule, center=<2 0 0>, rotation=<rot 90 0 0>}]
 base_necklace_out [{shapeType=Capsule, length=0.2, radius=0.2, shapeName=base_necklace_out_capsule, center=<2 0 0>, rotation=<rot 90 0 0>}]
 chest_helper [{shapeType=Capsule, length=12.2327, radius=8.1857, shapeName=chest_helper_capsule, center=<2.2397 7.9888 0>, rotation=<rot 0 -90 0>}]
 dyn_jacket_ornament_l [{
   shapeType=Capsule
   length=4.1284
   radius=1.3531
   shapeName=dyn_jacket_ornament_r_capsule
   center=<5.1 -1.2 0>
   rotation=<rot -0.0436 -85.0824 -89.4438>
  }]
 dyn_jacket_ornament_r [{
   shapeType=Capsule
   length=4.1284
   radius=1.3531
   shapeName=dyn_jacket_ornament_r_capsule
   center=<-5.1 1.2 0>
   rotation=<rot 0.052 -85.0824 -89.4445>
  }]
 dyn_neck_back [{shapeType=Capsule, length=6, radius=4, shapeName=dyn_neck_front_l_capsule, center=<6 -3 0>}]
 dyn_neck_front_l [{shapeType=Capsule, length=6, radius=4, shapeName=dyn_neck_front_l_capsule, center=<5 -2 2.5>, rotation=<rot 25.7197 0 -1.7262>}]
 dyn_neck_front_r [{shapeType=Capsule, length=6, radius=4, shapeName=dyn_neck_front_l_capsule, center=<4 -2 -3.5>, rotation=<rot -26.2281 -3.3852 -4.9253>}]
 dyn_neck_mid_l [{shapeType=Capsule, length=6, radius=4, shapeName=dyn_neck_front_l_capsule, center=<5 -2 2.5>, rotation=<rot -1.4313 -0.3629 14.2221>}]
 dyn_neck_mid_r [{shapeType=Capsule, length=6, radius=4, shapeName=dyn_neck_front_l_capsule, center=<5 -2 2>, rotation=<rot 13.7701 3.1222 11.0827>}]
 dyn_neck_midback_r [{shapeType=Capsule, length=6, radius=4, shapeName=dyn_neck_front_l_capsule, center=<4.9356 -4.5 -0.0852>, rotation=<rot 4.5788 0 0>}]
 dyn_necklace_in_01 [{shapeType=Capsule, length=2, radius=1, shapeName=dyn_necklace_in_01_capsule, center=<2.0246 -0.9848 0>, rotation=<rot 90 3.6648 0>}]
 dyn_necklace_in_02 [{shapeType=Capsule, length=2, radius=1, shapeName=dyn_necklace_in_02_capsule, center=<2.0105 -0.7081 0>, rotation=<rot 90 5.5318 0>}]
 dyn_necklace_in_03 [{shapeType=Capsule, length=2, radius=1, shapeName=dyn_necklace_in_03_capsule, center=<2.0292 -0.1046 0>, rotation=<rot 90 15.5749 0>}]
 dyn_necklace_out_01 [{shapeType=Capsule, length=2, radius=1, shapeName=dyn_necklace_out_01_capsule, center=<2 -1.0856 0>, rotation=<rot 90 0 0>}]
 dyn_necklace_out_02 [{shapeType=Capsule, length=2, radius=1, shapeName=dyn_necklace_out_02_capsule, center=<2 -0.9086 0>, rotation=<rot 90 0 0>}]
 dyn_necklace_out_03 [{shapeType=Capsule, length=2, radius=1, shapeName=dyn_necklace_out_03_capsule, center=<2.0237 -0.106 0>, rotation=<rot 90 12.597 0>}]
 head [{shapeType=Capsule, length=6.611, radius=8.4998, shapeName=head_capsule, center=<3.3367 -1.2447 0>, rotation=<rot 90 -5.0868 0>}]
 neck_01 [{shapeType=Capsule, length=1, radius=6, shapeName=neck_01_capsule, center=<0 1.57 0>, rotation=<rot 90 0 0>}]
 neck_02 [{shapeType=Capsule, length=1, radius=6, shapeName=neck_02_capsule, center=<0 1.57 0>, rotation=<rot 90 0 0>}]
 spine_04 [
  {shapeType=Capsule, length=20.6429, radius=20, shapeName=spine_04_capsule, center=<-5.0022 -9.6559 0.1271>, rotation=<rot 0.6424 -90 -0.0051>}
  {shapeType=Capsule, length=15.3192, radius=15.538, shapeName=spine_04_capsule, rotation=<rot -90 0 0>}
 ]
 spine_05 [{
   shapeType=Capsule
   length=16.2685
   radius=13.3867
   shapeName=spine_05_capsule
   center=<-3.25 -2.4729 0.1586>
   rotation=<rot -1.3995 -90 -0.3297>
  }]
}
constraints [
 {bone1Name=dyn_neck_front_l, bone2Name=base_neck_front_l, swing1LimitDegrees=20, swing2LimitDegrees=1, twistLimitDegrees=1}
 {bone1Name=dyn_neck_front_r, bone2Name=base_neck_front_r, swing1LimitDegrees=17, swing2LimitDegrees=1, twistLimitDegrees=0.1}
 {bone1Name=dyn_neck_mid_l, bone2Name=base_neck_mid_l, swing1LimitDegrees=15, swing2LimitDegrees=5, twistLimitDegrees=1}
 {bone1Name=dyn_neck_back, bone2Name=base_neck_back, swing1LimitDegrees=15, swing2LimitDegrees=5, twistLimitDegrees=1}
 {
  bone1Name=neck_02
  bone2Name=neck_01
  swing1LimitDegrees=45
  swing1Motion=Free
  swing2LimitDegrees=45
  swing2Motion=Free
  twistLimitDegrees=45
  twistMotion=Free
 }
 {
  bone1Name=head
  bone2Name=neck_02
  swing1LimitDegrees=45
  swing1Motion=Free
  swing2LimitDegrees=45
  swing2Motion=Free
  twistLimitDegrees=45
  twistMotion=Free
 }
 {bone1Name=base_neck_midback_r, bone2Name=neck_01, swing1LimitDegrees=45, swing2LimitDegrees=45, twistLimitDegrees=45}
 {bone1Name=dyn_neck_midback_r, bone2Name=base_neck_midback_r, swing1LimitDegrees=15, swing2LimitDegrees=5, twistLimitDegrees=1}
 {
  bone1Name=dyn_necklace_in_01
  bone2Name=base_necklace_in
  swing1LimitDegrees=15
  swing2LimitDegrees=5
  twistLimitDegrees=45
  twistMotion=Locked
 }
 {
  bone1Name=dyn_necklace_in_02
  bone2Name=dyn_necklace_in_01
  swing1LimitDegrees=15
  swing2LimitDegrees=5
  twistLimitDegrees=45
  twistMotion=Locked
 }
 {
  bone1Name=dyn_necklace_in_03
  bone2Name=dyn_necklace_in_02
  swing1LimitDegrees=15
  swing2LimitDegrees=5
  twistLimitDegrees=45
  twistMotion=Locked
 }
 {
  bone1Name=dyn_necklace_out_01
  bone2Name=base_necklace_out
  swing1LimitDegrees=50
  swing2LimitDegrees=8
  twistLimitDegrees=45
  twistMotion=Locked
 }
 {
  bone1Name=dyn_necklace_out_02
  bone2Name=dyn_necklace_out_01
  swing1LimitDegrees=50
  swing2LimitDegrees=8
  twistLimitDegrees=45
  twistMotion=Locked
 }
 {
  bone1Name=dyn_necklace_out_03
  bone2Name=dyn_necklace_out_02
  swing1LimitDegrees=50
  swing2LimitDegrees=8
  twistLimitDegrees=45
  twistMotion=Locked
 }
 {
  bone1Name=neck_01
  bone2Name=spine_05
  swing1LimitDegrees=45
  swing1Motion=Free
  swing2LimitDegrees=45
  swing2Motion=Free
  twistLimitDegrees=45
  twistMotion=Free
 }
 {
  bone1Name=base_neck_front_l
  bone2Name=spine_05
  swing1LimitDegrees=45
  swing1Motion=Free
  swing2LimitDegrees=45
  swing2Motion=Free
  twistLimitDegrees=45
  twistMotion=Free
 }
 {
  bone1Name=base_neck_mid_l
  bone2Name=spine_05
  swing1LimitDegrees=45
  swing1Motion=Free
  swing2LimitDegrees=45
  swing2Motion=Free
  twistLimitDegrees=45
  twistMotion=Free
 }
 {
  bone1Name=base_neck_back
  bone2Name=spine_05
  swing1LimitDegrees=45
  swing1Motion=Free
  swing2LimitDegrees=45
  swing2Motion=Free
  twistLimitDegrees=45
  twistMotion=Free
 }
 {
  bone1Name=base_neck_mid_r
  bone2Name=spine_05
  swing1LimitDegrees=45
  swing1Motion=Free
  swing2LimitDegrees=45
  swing2Motion=Free
  twistLimitDegrees=45
  twistMotion=Free
 }
 {
  bone1Name=base_neck_front_r
  bone2Name=spine_05
  swing1LimitDegrees=45
  swing1Motion=Free
  swing2LimitDegrees=45
  swing2Motion=Free
  twistLimitDegrees=45
  twistMotion=Free
 }
 {
  bone1Name=base_necklace_in
  bone2Name=spine_05
  swing1LimitDegrees=45
  swing1Motion=Free
  swing2LimitDegrees=45
  swing2Motion=Free
  twistLimitDegrees=45
  twistMotion=Free
 }
 {
  bone1Name=base_necklace_out
  bone2Name=spine_05
  swing1LimitDegrees=45
  swing1Motion=Free
  swing2LimitDegrees=45
  swing2Motion=Free
  twistLimitDegrees=45
  twistMotion=Free
 }
 {
  bone1Name=chest_helper
  bone2Name=spine_05
  swing1LimitDegrees=45
  swing1Motion=Free
  swing2LimitDegrees=45
  swing2Motion=Free
  twistLimitDegrees=45
  twistMotion=Free
 }
 {bone1Name=dyn_neck_mid_r, bone2Name=base_neck_mid_r, swing1LimitDegrees=20, swing2LimitDegrees=5, twistLimitDegrees=1, twistMotion=Locked}
 {
  bone1Name=dyn_jacket_ornament_l
  bone2Name=base_jacket_ornament_l
  swing1LimitDegrees=45
  swing2LimitDegrees=25
  twistLimitDegrees=45
  twistMotion=Locked
 }
 {
  bone1Name=dyn_jacket_ornament_r
  bone2Name=base_jacket_ornament_r
  swing1LimitDegrees=45
  swing2LimitDegrees=25
  twistLimitDegrees=45
  twistMotion=Locked
 }
 {
  bone1Name=spine_05
  bone2Name=spine_04
  swing1LimitDegrees=45
  swing1Motion=Free
  swing2LimitDegrees=45
  swing2Motion=Free
  twistLimitDegrees=45
  twistMotion=Free
 }
 {
  bone1Name=base_jacket_ornament_l
  bone2Name=spine_04
  swing1LimitDegrees=45
  swing1Motion=Free
  swing2LimitDegrees=45
  swing2Motion=Free
  twistLimitDegrees=45
  twistMotion=Free
 }
 {
  bone1Name=base_jacket_ornament_r
  bone2Name=spine_04
  swing1LimitDegrees=45
  swing1Motion=Free
  swing2LimitDegrees=45
  swing2Motion=Free
  twistLimitDegrees=45
  twistMotion=Free
 }
]
