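# Animation Blueprint for the AnvilAlarmMoth medium male body cosmetic, adding secondary motion to a pose copied from the attached parent mesh.
# Weighted constraints drive neck helper bones, then a Control Rig receives movement states and a front-end idle weight.
# RigidBody and AnimDynamics simulate jacket, belt, light, bracelet and zipper bones, among additive bone offsets.
# The event graph computes gravity-override physics parameters, idle weights and reduced front-end dynamics alpha.
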AnimBlueprint /BRCosmetics/Characters/Player/Male/Medium/Bodies/M_MED_AnvilAlarmMoth/Meshes/M_MED_AnvilAlarmMoth_AnimBP
target_skeleton=&/M_MED_AnvilAlarmMoth_Skeleton
graphs [
 AnimGraph {nodes [
   n1 = "Constraint - base_neck_front_r (PPP)"!(
    ComponentPose=Local To Component(LocalPose=Copy Pose From Mesh() {bUseAttachedParent=true})
    Alpha=1
    ConstraintWeights_0=0.5
    ConstraintWeights_1:float=0.25
    ConstraintWeights_2:float=0.25
   ) -> (Pose) {
    alphaBoolBlend {blendOption=Linear}
    boneToModify {boneName=base_neck_front_r}
    constraintSetup [{targetBone {boneName=clavicle_r}}, {targetBone {boneName=spine_05}}, {targetBone {boneName=neck_02}}]
    constraintWeights [1,1,1]
   }
   n2 = "Constraint - base_neck_mid_r (PPP)"!(ComponentPose=@n1, Alpha=1, ConstraintWeights_0=0.4, ConstraintWeights_1:float=0.3, ConstraintWeights_2:float=0.3) -> (Pose) {
    alphaBoolBlend {blendOption=Linear}
    boneToModify {boneName=base_neck_mid_r}
    constraintSetup [{targetBone {boneName=clavicle_r}}, {targetBone {boneName=spine_05}}, {targetBone {boneName=neck_02}}]
    constraintWeights [1,1,1]
   }
   n3 = "Constraint - base_neck_midback_r (PPP)"!(ComponentPose=@n2, Alpha=1, ConstraintWeights_0=0.3, ConstraintWeights_1:float=0.3, ConstraintWeights_2:float=0.4) -> (Pose) {
    alphaBoolBlend {blendOption=Linear}
    boneToModify {boneName=base_neck_midback_r}
    constraintSetup [{targetBone {boneName=clavicle_r}}, {targetBone {boneName=spine_05}}, {targetBone {boneName=neck_02}}]
    constraintWeights [1,1,1]
   }
   n4 = "Constraint - base_neck_back (PP)"!(ComponentPose=@n3, Alpha=1, ConstraintWeights_0=0.7, ConstraintWeights_1:float=0.3) -> (Pose) {
    alphaBoolBlend {blendOption=Linear}
    boneToModify {boneName=base_neck_back}
    constraintSetup [{targetBone {boneName=spine_05}}, {targetBone {boneName=neck_02}}]
    constraintWeights [1,1]
   }
   n5 = "Constraint - base_neck_mid_l (PPP)"!(ComponentPose=@n4, Alpha=1, ConstraintWeights_0=0.4, ConstraintWeights_1:float=0.3, ConstraintWeights_2:float=0.3) -> (Pose) {
    alphaBoolBlend {blendOption=Linear}
    boneToModify {boneName=base_neck_mid_l}
    constraintSetup [{targetBone {boneName=clavicle_l}}, {targetBone {boneName=spine_05}}, {targetBone {boneName=neck_02}}]
    constraintWeights [1,1,1]
   }
   n6 = "Constraint - base_neck_front_l (PPP)"!(ComponentPose=@n5, Alpha=1, ConstraintWeights_0=0.5, ConstraintWeights_1:float=0.25, ConstraintWeights_2:float=0.25) -> (Pose) {
    alphaBoolBlend {blendOption=Linear}
    boneToModify {boneName=base_neck_front_l}
    constraintSetup [{targetBone {boneName=clavicle_l}}, {targetBone {boneName=spine_05}}, {targetBone {boneName=neck_02}}]
    constraintWeights [1,1,1]
   }
   n7 = Control Rig!(
    Source=Component To Local(ComponentPose=@n6)
    Alpha
    isCrouching:bool=Get bIsCrouching()
    isMoving2D:bool=Get bIsMoving2D()
    isSliding:bool=Get bIsSliding()
    isParachuteOpened:bool=Get bIsParachuteOpened()
    isDBNO:bool=Get bIsDBNO()
    isTargeting:bool=Get bIsTargeting()
    isJumping:bool=Get bIsJumping()
    isRigidBodyEnabled:bool=Get bRigidBodyEnabled()
    isSkydiving:bool=Get bIsSkyDiving()
    isFrontEnd:bool=Get bIsInFrontEnd()
    FEIdle:double=Get FE Idle() -> (FE Idle:double)
   ) -> (Pose) {controlRigAssetReference {blueprintRigClass=&/M_MED_AnvilAlarmMoth_CtrlRig.M_MED_AnvilAlarmMoth_CtrlRig_C}}
   n8 = RigidBody!(
    ComponentPose=Local To Component(LocalPose=@n7)
    Alpha
    OverrideWorldGravity:Vector="0,0,0"
    ExternalForce
    SimSpaceSettings:Sim Space Settings Structure=(WorldAlpha=1,VelocityScaleZ=0.1,DampingAlpha=0,MaxLinearVelocity=100,MaxAngularVelocity=8,MaxLinearAcceleration=500,MaxAngularAcceleration=30,ExternalLinearDragV=(X=0.2,Y=0.2,Z=0.2),ExternalLinearVelocity=(X=0,Y=0,Z=0),ExternalAngularVelocity=(X=0,Y=0,Z=0))
   ) -> (Pose) {
    bClampLinearTranslationLimitToRefPose=true
    bOverrideWorldGravity=true
    overridePhysicsAsset=&/M_MED_AnvilAlarmMoth_Chain_Physics
    simulationSpace=BaseBoneSpace
    alphaBoolBlend {blendOption=Linear}
    baseBoneRef {boneName=thigh_r}
    componentAppliedLinearAccClamp=<3000 3000 3000>
    componentLinearAccScale=<0.6 0.6 0.6>
    componentLinearVelScale=<0.6 0.6 0.6>
    simSpaceSettings {
     dampingAlpha=0
     maxAngularAcceleration=30
     maxAngularVelocity=8
     maxLinearAcceleration=500
     maxLinearVelocity=100
     velocityScaleZ=0.1
     worldAlpha=1
     externalLinearDragV=<0.2 0.2 0.2>
    }
   }
   n9 = RigidBody!(ComponentPose=@n8, Alpha, OverrideWorldGravity="0,0,0", ExternalForce) -> (Pose) {
    bClampLinearTranslationLimitToRefPose=true
    bOverrideWorldGravity=true
    overridePhysicsAsset=&/M_MED_AnvilAlarmMoth_Up_Physics
    simulationSpace=BaseBoneSpace
    alphaBoolBlend {blendOption=Linear}
    baseBoneRef {boneName=spine_05}
    componentAppliedLinearAccClamp=<3000 3000 3000>
    componentLinearAccScale {x=0.2}
    componentLinearVelScale {x=0.2}
    overrideWorldGravity {z=-980}
    simSpaceSettings {maxAngularAcceleration=100, maxAngularVelocity=5, maxLinearAcceleration=500, maxLinearVelocity=100, velocityScaleZ=0, worldAlpha=1}
   }
   n10 = RigidBody(ComponentPose=@n9, OverrideWorldGravity="0,0,0") {
    bClampLinearTranslationLimitToRefPose=true
    bOverrideWorldGravity=true
    overridePhysicsAsset=&/M_MED_AnvilAlarmMoth_LowerJacket_Physics
    simulationSpace=BaseBoneSpace
    alphaBoolBlend {blendOption=Linear}
    baseBoneRef {boneName=spine_01}
    componentAppliedLinearAccClamp=<3000 1000 1000>
    componentLinearAccScale {x=0.1}
    componentLinearVelScale {x=0.2}
    overrideWorldGravity {z=-980}
    simSpaceSettings {maxAngularAcceleration=100, maxAngularVelocity=5, maxLinearAcceleration=500, maxLinearVelocity=100, velocityScaleZ=0, worldAlpha=1}
   }
   n11 = "Transform (Modify) Bone - Bone: dyn_neck_midback_r"(ComponentPose=@n10, Alpha=1, Translation="0,0,0", Rotation="0,0,0", Scale="1,1,1") {
    rotationMode=BMM_Additive
    rotationSpace=BCS_BoneSpace
    translationMode=BMM_Additive
    translationSpace=BCS_BoneSpace
    alphaScaleBiasClamp {bInterpResult=true, interpSpeedDecreasing=5, interpSpeedIncreasing=5}
    boneToModify {boneName=dyn_neck_midback_r}
   }
   n12 = "Transform (Modify) Bone - Bone: base_neck_back"(ComponentPose=@n11, Alpha=1, Translation="0,0,0", Rotation="0,-5,10", Scale="1,1,1") {
    rotationMode=BMM_Additive
    rotationSpace=BCS_BoneSpace
    translationMode=BMM_Additive
    translationSpace=BCS_BoneSpace
    alphaScaleBiasClamp {bInterpResult=true, interpSpeedDecreasing=5, interpSpeedIncreasing=5}
    boneToModify {boneName=base_neck_back}
    rotation {roll=10, yaw=-5}
   }
   n13 = "Transform (Modify) Bone - Bone: dyn_neck_mid_l"(ComponentPose=@n12, Alpha=1, Translation="0,0,0", Rotation="0,-10,-20", Scale="1,1,1") {
    rotationMode=BMM_Additive
    rotationSpace=BCS_BoneSpace
    translationMode=BMM_Additive
    translationSpace=BCS_BoneSpace
    alphaScaleBiasClamp {bInterpResult=true, interpSpeedDecreasing=5, interpSpeedIncreasing=5}
    boneToModify {boneName=dyn_neck_mid_l}
    rotation {roll=-20, yaw=-10}
   }
   n14 = Anim Dynamics - Bone: dyn_belt_01(ComponentPose=@n13, Alpha=1, ExternalForce="0,0,0") {
    angularDampingOverride=1
    angularSpringConstant=10000
    bAngularSpring=true
    bOverrideAngularDamping=true
    bUseGravityOverride=true
    gravityScale=0.35
    simulationSpace=BoneRelative
    boundBone {boneName=dyn_belt_01}
    componentAppliedLinearAccClamp=<10000 10000 10000>
    componentLinearAccScale=<1 1 1>
    componentLinearVelScale=<1 1 1>
    gravityOverride {z=-980}
    physicsBodyDefinitions [{
      boundBone {boneName=dyn_belt_01}
      boxExtents=<20 30 30>
      constraintSetup {angularLimitsMax=<0 2 0>, angularLimitsMin=<0 -2 -2>, angularTarget=<1 0 0>}
      localJointOffset=<4 0 0>
     }]
    relativeSpaceBone {boneName=pelvis}
    simSpaceSettings {maxAngularAcceleration=35, maxAngularVelocity=7, simSpaceAngularAlpha=1}
   }
   n15 = Anim Dynamics - Bone: dyn_belt_02(ComponentPose=@n14, Alpha=1, ExternalForce="0,0,0") {
    angularDampingOverride=1
    angularSpringConstant=10000
    bAngularSpring=true
    bOverrideAngularDamping=true
    bUseGravityOverride=true
    gravityScale=0.35
    simulationSpace=BoneRelative
    boundBone {boneName=dyn_belt_02}
    componentAppliedLinearAccClamp=<10000 10000 10000>
    componentLinearAccScale=<1 1 1>
    componentLinearVelScale=<1 1 1>
    gravityOverride {z=-980}
    physicsBodyDefinitions [{
      boundBone {boneName=dyn_belt_02}
      boxExtents=<20 30 30>
      constraintSetup {angularLimitsMax=<0 2 0>, angularLimitsMin=<0 -2 -2>, angularTarget=<1 0 0>}
      localJointOffset=<4 0 0>
     }]
    relativeSpaceBone {boneName=pelvis}
    simSpaceSettings {maxAngularAcceleration=35, maxAngularVelocity=7, simSpaceAngularAlpha=1}
   }
   n16 = Anim Dynamics - Bone: dyn_light_01(ComponentPose=@n15, Alpha=1, ExternalForce="0,0,0") {
    angularDampingOverride=1
    angularSpringConstant=12000
    bAngularSpring=true
    bOverrideAngularDamping=true
    bUseGravityOverride=true
    gravityScale=0.35
    simulationSpace=BoneRelative
    boundBone {boneName=dyn_light_01}
    componentAppliedLinearAccClamp=<10000 10000 10000>
    componentLinearAccScale=<1 1 1>
    componentLinearVelScale=<1 1 1>
    gravityOverride {z=-980}
    physicsBodyDefinitions [{
      boundBone {boneName=dyn_light_01}
      boxExtents=<20 30 30>
      constraintSetup {angularLimitsMax=<0 0 10>, angularLimitsMin=<0 -15 -15>, angularTarget=<1 0 0>}
      localJointOffset=<2 0 0>
     }]
    relativeSpaceBone {boneName=base_light}
    simSpaceSettings {maxAngularAcceleration=35, maxAngularVelocity=7, simSpaceAngularAlpha=1}
   }
   n17 = Anim Dynamics - Bone: dyn_light_02(ComponentPose=@n16, Alpha=1, ExternalForce="0,0,0") {
    angularDampingOverride=1
    angularSpringConstant=12000
    bAngularSpring=true
    bOverrideAngularDamping=true
    bUseGravityOverride=true
    gravityScale=0.35
    simulationSpace=BoneRelative
    boundBone {boneName=dyn_light_02}
    componentAppliedLinearAccClamp=<10000 10000 10000>
    componentLinearAccScale=<1 1 1>
    componentLinearVelScale=<1 1 1>
    gravityOverride {z=-980}
    physicsBodyDefinitions [{
      boundBone {boneName=dyn_light_02}
      boxExtents=<20 30 30>
      constraintSetup {angularLimitsMax=<0 0 10>, angularLimitsMin=<0 -15 -15>, angularTarget=<1 0 0>}
      localJointOffset=<2 0 0>
     }]
    relativeSpaceBone {boneName=base_light}
    simSpaceSettings {maxAngularAcceleration=35, maxAngularVelocity=7, simSpaceAngularAlpha=1}
   }
   n18 = Anim Dynamics - Bone: dyn_light_03(ComponentPose=@n17, Alpha=1, ExternalForce="0,0,0") {
    angularDampingOverride=1
    angularSpringConstant=12000
    bAngularSpring=true
    bOverrideAngularDamping=true
    bUseGravityOverride=true
    gravityScale=0.35
    simulationSpace=BoneRelative
    boundBone {boneName=dyn_light_03}
    componentAppliedLinearAccClamp=<10000 10000 10000>
    componentLinearAccScale=<1 1 1>
    componentLinearVelScale=<1 1 1>
    gravityOverride {z=-980}
    physicsBodyDefinitions [{
      boundBone {boneName=dyn_light_03}
      boxExtents=<20 30 30>
      constraintSetup {angularLimitsMax=<0 0 15>, angularLimitsMin=<0 -15 -15>, angularTarget=<1 0 0>}
      localJointOffset=<2 0 0>
     }]
    relativeSpaceBone {boneName=base_light}
    simSpaceSettings {maxAngularAcceleration=35, maxAngularVelocity=7, simSpaceAngularAlpha=1}
   }
   n19 = Anim Dynamics - Bone: dyn_bracelet_01(ComponentPose=@n18, Alpha=1, ExternalForce="0,0,0") {
    angularDampingOverride=1
    angularSpringConstant=20000
    bAngularSpring=true
    bOverrideAngularDamping=true
    bUseGravityOverride=true
    gravityScale=0.35
    simulationSpace=BoneRelative
    boundBone {boneName=dyn_bracelet_01}
    componentLinearAccScale=<1 1 1>
    componentLinearVelScale=<1 1 1>
    gravityOverride {z=-980}
    physicsBodyDefinitions [{
      boundBone {boneName=dyn_bracelet_01}
      boxExtents=<20 30 30>
      constraintSetup {angularLimitsMax=<0 0 7>, angularLimitsMin=<0 0 -4>, angularTarget=<1 0 0>}
      localJointOffset=<4 0 0>
     }]
    relativeSpaceBone {boneName=hand_l}
    simSpaceSettings {maxAngularAcceleration=35, maxAngularVelocity=7, simSpaceAngularAlpha=1}
   }
   n20 = Anim Dynamics - Bone: dyn_bracelet_02(ComponentPose=@n19, Alpha=1, ExternalForce="0,0,0") {
    angularDampingOverride=1
    angularSpringConstant=20000
    bAngularSpring=true
    bOverrideAngularDamping=true
    bUseGravityOverride=true
    gravityScale=0.35
    simulationSpace=BoneRelative
    boundBone {boneName=dyn_bracelet_02}
    componentLinearAccScale=<1 1 1>
    componentLinearVelScale=<1 1 1>
    gravityOverride {z=-980}
    physicsBodyDefinitions [{
      boundBone {boneName=dyn_bracelet_02}
      boxExtents=<20 30 30>
      constraintSetup {angularLimitsMax=<0 0 7>, angularLimitsMin=<0 0 -7>, angularTarget=<1 0 0>}
      localJointOffset=<4 0 0>
     }]
    relativeSpaceBone {boneName=hand_l}
    simSpaceSettings {maxAngularAcceleration=35, maxAngularVelocity=7, simSpaceAngularAlpha=1}
   }
   n21 = Anim Dynamics - Bone: dyn_bracelet_03(ComponentPose=@n20, Alpha=1, ExternalForce="0,0,0") {
    angularDampingOverride=1
    angularSpringConstant=20000
    bAngularSpring=true
    bOverrideAngularDamping=true
    bUseGravityOverride=true
    gravityScale=0.35
    simulationSpace=BoneRelative
    boundBone {boneName=dyn_bracelet_03}
    componentLinearAccScale=<1 1 1>
    componentLinearVelScale=<1 1 1>
    gravityOverride {z=-980}
    physicsBodyDefinitions [{
      boundBone {boneName=dyn_bracelet_03}
      boxExtents=<20 30 30>
      constraintSetup {angularLimitsMax=<0 0 7>, angularLimitsMin=<0 0 -7>, angularTarget=<1 0 0>}
      localJointOffset=<4 0 0>
     }]
    relativeSpaceBone {boneName=hand_l}
    simSpaceSettings {maxAngularAcceleration=35, maxAngularVelocity=7, simSpaceAngularAlpha=1}
   }
   n22 = Anim Dynamics - Bone: dyn_jacketZipper_l!(ComponentPose=@n21, Alpha=1, GravityOverride:Vector="0,0,-1200", ExternalForce="0,0,0") -> (Pose) {
    angularDampingOverride=1
    angularSpringConstant=3000
    bAngularSpring=true
    bOverrideAngularDamping=true
    bUseGravityOverride=true
    gravityScale=0.35
    simulationSpace=RootRelative
    boundBone {boneName=dyn_jacketZipper_l}
    componentAppliedLinearAccClamp=<12000 12000 12000>
    componentLinearAccScale=<1 1 1>
    componentLinearVelScale=<1 1 1>
    gravityOverride {z=-1200}
    physicsBodyDefinitions [{
      boundBone {boneName=dyn_jacketZipper_l}
      boxExtents=<15 20 20>
      constraintSetup {angularLimitsMax=<0 7 0>, angularLimitsMin=<0 -30 -150>, angularTarget=<1 0 0>}
      localJointOffset=<7 0 0>
     }]
    relativeSpaceBone {boneName=lowerarm_l}
    simSpaceSettings {maxAngularAcceleration=1000, maxAngularVelocity=1000, simSpaceAngularAlpha=1}
   }
   n23 = Anim Dynamics - Bone: dyn_jacketRing_l!(ComponentPose=@n22, Alpha=1, GravityOverride:Vector="0,0,-1200", ExternalForce="0,0,0") -> (Pose) {
    angularDampingOverride=1
    angularSpringConstant=3000
    bAngularSpring=true
    bOverrideAngularDamping=true
    bUseGravityOverride=true
    gravityScale=0.35
    simulationSpace=RootRelative
    boundBone {boneName=dyn_jacketRing_l}
    componentAppliedLinearAccClamp=<12000 12000 12000>
    componentLinearAccScale=<1 1 1>
    componentLinearVelScale=<1 1 1>
    gravityOverride {z=-1200}
    physicsBodyDefinitions [{
      boundBone {boneName=dyn_jacketRing_l}
      boxExtents=<15 20 20>
      constraintSetup {angularLimitsMax=<2 45 10>, angularLimitsMin=<-20 -100 -40>, angularTarget=<1 0 0>}
      localJointOffset=<7 0 0>
     }]
    planarLimits [{drivingBone {boneName=lowerarm_twist_01_l}, planeTransform {location=<0 0 4>, rotation=<rot -3 0 0>}}]
    relativeSpaceBone {boneName=lowerarm_l}
    simSpaceSettings {maxAngularAcceleration=1000, maxAngularVelocity=1000, simSpaceAngularAlpha=1}
   }
   n24 = Anim Dynamics - Bone: dyn_jacketZipper_r!(ComponentPose=@n23, Alpha=1, GravityOverride:Vector="0,0,-1200", ExternalForce="0,0,0") -> (Pose) {
    angularDampingOverride=1
    angularSpringConstant=3000
    bAngularSpring=true
    bOverrideAngularDamping=true
    bUseGravityOverride=true
    gravityScale=0.35
    simulationSpace=RootRelative
    boundBone {boneName=dyn_jacketZipper_r}
    componentAppliedLinearAccClamp=<12000 12000 12000>
    componentLinearAccScale=<1 1 1>
    componentLinearVelScale=<1 1 1>
    gravityOverride {z=-1200}
    physicsBodyDefinitions [{
      boundBone {boneName=dyn_jacketZipper_r}
      boxExtents=<15 20 20>
      constraintSetup {angularLimitsMax=<0 10 0>, angularLimitsMin=<0 -30 -150>, angularTarget=<1 0 0>}
      localJointOffset=<-7 0 0>
     }]
    relativeSpaceBone {boneName=lowerarm_r}
    simSpaceSettings {maxAngularAcceleration=1000, maxAngularVelocity=1000, simSpaceAngularAlpha=1}
   }
   n25 = Anim Dynamics - Bone: dyn_jacketRing_r!(ComponentPose=@n24, Alpha=1, GravityOverride:Vector="0,0,-1200", ExternalForce="0,0,0") -> (Pose) {
    angularDampingOverride=1
    angularSpringConstant=3000
    bAngularSpring=true
    bOverrideAngularDamping=true
    bUseGravityOverride=true
    gravityScale=0.35
    simulationSpace=RootRelative
    boundBone {boneName=dyn_jacketRing_r}
    componentAppliedLinearAccClamp=<12000 12000 12000>
    componentLinearAccScale=<1 1 1>
    componentLinearVelScale=<1 1 1>
    gravityOverride {z=-1200}
    physicsBodyDefinitions [{
      boundBone {boneName=dyn_jacketRing_r}
      boxExtents=<15 20 20>
      constraintSetup {angularLimitsMax=<10 55 10>, angularLimitsMin=<-2 -100 -75>, angularTarget=<1 0 0>}
      localJointOffset=<-7 0 0>
     }]
    planarLimits [{drivingBone {boneName=lowerarm_twist_01_r}, planeTransform {location=<0 0 -4>, rotation=<rot 3 180 180>}}]
    relativeSpaceBone {boneName=lowerarm_r}
    simSpaceSettings {maxAngularAcceleration=1000, maxAngularVelocity=1000, simSpaceAngularAlpha=1}
   }
   n26 = "Transform (Modify) Bone - Bone: base_necklace_out"(ComponentPose=@n25, Alpha=1, Translation="0,0,0", Rotation="0.419,3.559,0.499", Scale="1,1,1") {
    rotationMode=BMM_Additive
    rotationSpace=BCS_BoneSpace
    scaleMode=BMM_Additive
    scaleSpace=BCS_BoneSpace
    translationMode=BMM_Additive
    translationSpace=BCS_BoneSpace
    boneToModify {boneName=base_necklace_out}
    rotation=<rot 0.419 3.559 0.499>
   }
   n27 = "Transform (Modify) Bone - Bone: base_necklace_in"(ComponentPose=@n26, Alpha=1, Translation="0,0,0", Rotation="0.487,5,-0.844", Scale="1,1,1") {
    rotationMode=BMM_Additive
    rotationSpace=BCS_BoneSpace
    scaleMode=BMM_Additive
    scaleSpace=BCS_BoneSpace
    translationMode=BMM_Additive
    translationSpace=BCS_BoneSpace
    boneToModify {boneName=base_necklace_in}
    rotation=<rot 0.487 5 -0.844>
   }
   Output Pose(Result=Component To Local(ComponentPose=@n27)) {layerGroup=None}
  ]}
 EventGraph {nodes [
   Event BlueprintUpdateAnimation()
   > n1 = Sequence()
   > n2 = "Gravity Override State Function (Parameter Asset) (Nativized)"(
    CharacterPartAnimInstance=Self-Reference()
    GameTimeInSeconds=GetGameTimeInSeconds()
    DebugJointName=spine_05
    Parameters=&/M_MED_AnvilAlarmMoth_Params
   )
   > Set GravityOverride(GravityOverride=@n2)
   > Set RBAN Settings Out(RBAN Settings Out=@n2.OutRigidBodyAnimNodeParameters)
   > n3 = Branch(Condition=Get bIsInFrontEnd())
   > Set AlphaDynamics(AlphaDynamics:double=0.3) -> (Output_Get:double)
   Set AlphaDynamics(execute=@n3.else, AlphaDynamics=1)
   n4 = AND Boolean!(
    A=Get bIsInFrontEnd()
    B=NOT Boolean(A=Get bIsMoving2D())
    C:bool=NOT Boolean(A=Get bIsPlayingEmote())
    D:bool=NOT Boolean(A=Get bIsSkyDiving())
   ) -> (ReturnValue)
   Set FE Idle(execute=@n1.then_1, FE Idle:double="To Float (Boolean)"(InBool=@n4)) -> (Output_Get:double)
   n5 = Get bRigidBodyEnabled()
   > Set FE Idle RBAN(FE Idle RBAN:double="To Float (Boolean)"(InBool=AND Boolean!(A=@n4, B=@n5) -> (ReturnValue))) -> (Output_Get:double)
  ]}
]
variables [
 AlphaDynamics {type=number}
 FE Idle {type=number}
 FE Idle RBAN {type=number}
 Gravity Override Params {type=GravityOverrideParamsStruct}
 GravityOverride {type=Vector}
 RBAN Settings Out {type=FortRigidBodyAnimNodeParameters}
]
