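# Control Rig for the AnvilAlarmMoth male face-accessory part that applies corrective offsets to the base hair bones parenting the dyn hair bones.
# Spherical pose readers on the head weight several corrections, one multiplied by a right-upperarm reader for two-handed upward aiming.
# Parachute, skydive-float, half-strength frontend and parachute-without-rigid-body flags weight other offsets.
# Two offsets always apply at full weight, leaving the last head reader's output unused.
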ControlRigBlueprint /BRCosmetics/Characters/Player/Male/Medium/Bodies/M_MED_AnvilAlarmMoth/Meshes/Parts/M_MED_AnvilAlarmMoth_FaceAcc_CtrlRig
graphs [
 RigVMFunctionLibrary {links [], nodes []}
 RigVMModel {nodes [
   RigUnit_BeginExecution = &RigUnit_BeginExecution()
   n1 = &RigUnit_SphericalPoseReader(
    ActiveRegionScaleFactors=(PositiveWidth=1,NegativeWidth=0.5,PositiveHeight=0.75,NegativeHeight=0.75)
    ActiveRegionSize=0.1
    Debug=(bDrawDebug=false,bDraw2D=False,bDrawLocalAxes=False,DebugScale=25,DebugSegments=20,DebugThickness=0.25)
    DriverAxis=(X=0,Y=1,Z=0)
    DriverItem=(Type=Bone,Name="head")
    FalloffRegionScaleFactors=(PositiveWidth=1,NegativeWidth=0.5,PositiveHeight=1,NegativeHeight=1)
    FalloffSize=0.2
    FlipHeightScaling=False
    FlipWidthScaling=false
    OptionalParentItem=(Type=Bone,Name="None")
    RotationOffset=(X=90,Y=0,Z=-30)
    ExecutePin=@RigUnit_BeginExecution.ExecutePin
   )
   > RigVMVariableNode(Value=@n1.OutputParam) {variable=head_SPR_look_up_alpha}
   > n2 = &RigUnit_SphericalPoseReader(
    ActiveRegionScaleFactors=(PositiveWidth=1,NegativeWidth=1,PositiveHeight=1,NegativeHeight=1)
    ActiveRegionSize=0.025
    Debug=(bDrawDebug=false,bDraw2D=False,bDrawLocalAxes=False,DebugScale=25,DebugSegments=20,DebugThickness=0.25)
    DriverAxis=(X=-1,Y=0,Z=0)
    DriverItem=(Type=Bone,Name="upperarm_r")
    FalloffRegionScaleFactors=(PositiveWidth=1,NegativeWidth=1,PositiveHeight=1,NegativeHeight=1)
    FalloffSize=0.1
    FlipHeightScaling=False
    FlipWidthScaling=False
    OptionalParentItem=(Type=Bone,Name="clavicle_r")
    RotationOffset=(X=-115,Y=0,Z=34)
   )
   > RigVMVariableNode(Value=@n2.OutputParam) {variable=upperarm_r_SPR_aim_up_two_hands_alpha}
   Multiply = &RigVMFunction_MathDoubleMul(A=RigVMVariableNode() {variable=head_SPR_look_up_alpha}.Value, B=RigVMVariableNode() {variable=upperarm_r_SPR_aim_up_two_hands_alpha}.Value)
   > RigVMVariableNode(Value=@Multiply.Result) {variable=look_up_and_left_two_hands_multiplied_alpha}
   > &RigUnit_ModifyTransforms(
    ItemToModify="((Item=(Type=Bone,Name=\"base_hair_l_06\"),Transform=(Rotation=(X=0,Y=0,Z=0.766,W=0.6428),Translation=(X=0,Y=0,Z=0),Scale3D=(X=1,Y=1,Z=1))),(Item=(Type=Bone,Name=\"base_hair_l_04\"),Transform=(Rotation=(X=0,Y=0,Z=0.0436,W=0.999),Translation=(X=0,Y=0,Z=0),Scale ...(truncated)"
    WeightMinimum=0
    Weight=RigVMVariableNode() {variable=look_up_and_left_two_hands_multiplied_alpha}.Value
   )
   > n3 = &RigUnit_SphericalPoseReader(
    ActiveRegionScaleFactors=(PositiveWidth=1,NegativeWidth=0.5,PositiveHeight=0.75,NegativeHeight=0.75)
    ActiveRegionSize=0.1
    Debug=(bDrawDebug=false,bDraw2D=False,bDrawLocalAxes=False,DebugScale=25,DebugSegments=20,DebugThickness=0.25)
    DriverAxis=(X=0,Y=1,Z=0)
    DriverItem=(Type=Bone,Name="head")
    FalloffRegionScaleFactors=(PositiveWidth=1,NegativeWidth=0.5,PositiveHeight=1,NegativeHeight=1)
    FalloffSize=0.2
    FlipHeightScaling=False
    FlipWidthScaling=false
    OptionalParentItem=(Type=Bone,Name="None")
    RotationOffset=(X=100,Y=20,Z=-30)
   )
   > &RigUnit_ModifyTransforms(
    ItemToModify="((Item=(Type=Bone,Name=\"base_hair_l_04\"),Transform=(Rotation=(X=0.1695,Y=0.0376,Z=0.2132,W=0.9615),Translation=(X=0,Y=0,Z=0),Scale3D=(X=1,Y=1,Z=1))),(Item=(Type=Bone,Name=\"base_hair_l_03\"),Transform=(Rotation=(X=0,Y=0,Z=0.1908,W=0.9816),Translation=(X=0,Y= ...(truncated)"
    WeightMinimum=0
    Weight=@n3.OutputParam
   )
   > &RigUnit_ModifyTransforms(
    ItemToModify=((Item=(Type=Bone,Name="base_hair_r_04"),Transform=(Rotation=(X=0.0867,Y=0.0091,Z=0.1041,W=0.9907),Translation=(X=0,Y=0,Z=0),Scale3D=(X=1,Y=1,Z=1))))
    Weight=1
    WeightMinimum=0
   )
   RigVMFunction_MathBoolToFloat = &RigVMFunction_MathBoolToFloat(Value=RigVMVariableNode() {variable=isParachuteOpen}.Value)
   > &RigUnit_ModifyTransforms(
    ItemToModify="((Item=(Type=Bone,Name=\"base_hair_l_05\"),Transform=(Rotation=(X=0.171,Y=0.0302,Z=0.171,W=0.9698),Translation=(X=0,Y=0,Z=0),Scale3D=(X=1,Y=1,Z=1))),(Item=(Type=Bone,Name=\"base_hair_r_05\"),Transform=(Rotation=(X=0.171,Y=0.0302,Z=0.171,W=0.9698),Translation=( ...(truncated)"
    WeightMinimum=0
    Weight=@RigVMFunction_MathBoolToFloat.Result
   )
   > &RigUnit_ModifyTransforms(
    ItemToModify="((Item=(Type=Bone,Name=\"base_hair_back\"),Transform=(Rotation=(X=0,Y=0,Z=0,W=1),Translation=(X=0,Y=0,Z=0),Scale3D=(X=1,Y=1,Z=1))),(Item=(Type=Bone,Name=\"base_hair_r_06\"),Transform=(Rotation=(X=0,Y=0,Z=0,W=1),Translation=(X=0,Y=0,Z=0),Scale3D=(X=1,Y=1,Z=1))) ...(truncated)"
    WeightMinimum=0
    Weight=&RigVMFunction_MathBoolToFloat(Value=RigVMVariableNode() {variable=isSkyDiveFloating}.Value).Result
   )
   > n4 = &RigUnit_SphericalPoseReader(
    ActiveRegionScaleFactors=(PositiveWidth=1.5,NegativeWidth=0.8,PositiveHeight=1,NegativeHeight=1)
    ActiveRegionSize=0.1
    Debug=(bDrawDebug=false,bDraw2D=False,bDrawLocalAxes=False,DebugScale=25,DebugSegments=20,DebugThickness=0.25)
    DriverAxis=(X=0,Y=1,Z=0)
    DriverItem=(Type=Bone,Name="head")
    FalloffRegionScaleFactors=(PositiveWidth=1,NegativeWidth=0.35,PositiveHeight=1,NegativeHeight=1)
    FalloffSize=0.2
    FlipHeightScaling=False
    FlipWidthScaling=False
    OptionalParentItem=(Type=Bone,Name="None")
    RotationOffset=(X=120,Y=0,Z=-22)
   )
   > &RigUnit_ModifyTransforms(
    ItemToModify="((Item=(Type=Bone,Name=\"base_hair_l_04\"),Transform=(Rotation=(X=0,Y=0,Z=0.0872,W=0.9962),Translation=(X=0,Y=0,Z=0),Scale3D=(X=1,Y=1,Z=1))),(Item=(Type=Bone,Name=\"base_hair_l_05\"),Transform=(Rotation=(X=0,Y=0,Z=0.0175,W=0.9998),Translation=(X=0,Y=0.4,Z=0),S ...(truncated)"
    WeightMinimum=0
    Weight=@n4.OutputParam
   )
   RigVMFunction_MathBoolNot = &RigVMFunction_MathBoolNot(Value=RigVMVariableNode() {variable=isRigidBodyEnabled}.Value)
   RigVMFunction_MathBoolAnd = &RigVMFunction_MathBoolAnd(A=@RigVMFunction_MathBoolNot.Result, B=RigVMVariableNode() {variable=isParachuteOpen}.Value)
   > &RigUnit_ModifyTransforms(
    ItemToModify="((Item=(Type=Bone,Name=\"base_hair_back\"),Transform=(Rotation=(X=0,Y=0,Z=0.1305,W=0.9914),Translation=(X=0,Y=0,Z=0),Scale3D=(X=1,Y=1,Z=1))),(Item=(Type=Bone,Name=\"base_hair_r_06\"),Transform=(Rotation=(X=0,Y=0,Z=0.0872,W=0.9962),Translation=(X=0,Y=0,Z=0),Sca ...(truncated)"
    WeightMinimum=0
    Weight=&RigVMFunction_MathBoolToFloat(Value=@RigVMFunction_MathBoolAnd.Result).Result
   )
   > &RigUnit_SphericalPoseReader(
    ActiveRegionScaleFactors=(PositiveWidth=1,NegativeWidth=1,PositiveHeight=1,NegativeHeight=0.7)
    ActiveRegionSize=0.1
    Debug=(bDrawDebug=false,bDraw2D=False,bDrawLocalAxes=False,DebugScale=25,DebugSegments=20,DebugThickness=0.25)
    DriverAxis=(X=0,Y=1,Z=0)
    DriverItem=(Type=Bone,Name="head")
    FalloffRegionScaleFactors=(PositiveWidth=1,NegativeWidth=1,PositiveHeight=1,NegativeHeight=0.7)
    FalloffSize=0.2
    FlipHeightScaling=False
    FlipWidthScaling=False
    OptionalParentItem=(Type=Bone,Name="None")
    RotationOffset=(X=126,Y=0,Z=0)
   )
   > &RigUnit_ModifyTransforms(
    ItemToModify="((Item=(Type=Bone,Name=\"base_hair_l_03\"),Transform=(Rotation=(X=0,Y=0,Z=0.1736,W=0.9848),Translation=(X=0,Y=0,Z=0),Scale3D=(X=1,Y=1,Z=1))),(Item=(Type=Bone,Name=\"base_hair_l_04\"),Transform=(Rotation=(X=0,Y=0,Z=0.0872,W=0.9962),Translation=(X=0,Y=0,Z=0),Sca ...(truncated)"
    Weight=1
    WeightMinimum=0
   )
   n5 = RigVMVariableNode() {variable=isFrontend}
   > &RigUnit_ModifyTransforms(
    ItemToModify="((Item=(Type=Bone,Name=\"base_hair_r_04\"),Transform=(Rotation=(X=0,Y=0,Z=0.0872,W=0.9962),Translation=(X=0,Y=0,Z=0),Scale3D=(X=1,Y=1,Z=1))),(Item=(Type=Bone,Name=\"base_hair_r_05\"),Transform=(Rotation=(X=0,Y=0,Z=0.0872,W=0.9962),Translation=(X=0,Y=0,Z=0),Sca ...(truncated)"
    WeightMinimum=0
    Weight=&RigVMFunction_MathDoubleMul(B=0.5, A=&RigVMFunction_MathBoolToFloat(Value=@n5.Value).Result).Result
   )
  ]}
]
hierarchy [
 root {loc=<0 0 0>, rot=<0 0 0 1>}
 pelvis {parent=root, loc=<0 -0.7071 93.1191>, rot=<0.0127 0.707 -0.0127 -0.707>}
 spine_01 {parent=pelvis, loc=<6.7494 -0.6388 0>, rot=<0 0 0.034 -0.9994>}
 spine_02 {parent=spine_01, loc=<7.2518 -0.035 0>, rot=<0 0 0.0238 -0.9997>}
 spine_03 {parent=spine_02, loc=<6.0104 0 0>, rot=<0 0 0.0149 -0.9999>}
 spine_04 {parent=spine_03, loc=<8.1024 0 0>, rot=<0 0 -0.0132 -0.9999>}
 spine_05 {parent=spine_04, loc=<10.0146 0 0>, rot=<0 0 -0.0015 -1>}
 neck_01 {parent=spine_05, loc=<11.3138 0.1529 0>, rot=<0 0 -0.2003 -0.9797>}
 neck_02 {parent=neck_01, loc=<4.5627 0 0>, rot=<0 0 0.0504 -0.9987>}
 head {parent=neck_02, loc=<5.3914 0 0>, rot=<0 0 0.115 -0.9934>}
 base_bandana_top {parent=head, loc=<10.2846 1.7441 -8.0265>, rot=<0.7615 -0.4536 -0.4612 0.0409>}
 dyn_bandana_top {parent=base_bandana_top, loc=<0 0 0>, rot=<0 0 0 1>}
 base_bandana_low {parent=head, loc=<7.9984 2.589 -8.8289>, rot=<0.2354 -0.7057 -0.6677 0.0287>}
 dyn_bandana_low_01 {parent=base_bandana_low, loc=<0 0 0>, rot=<0 0 0 1>}
 base_hair_l_01 {parent=head, loc=<2.1148 5.1053 -6.4911>, rot=<0.0726 0.4556 0.8871 0.0162>}
 dyn_hair_l_01 {parent=base_hair_l_01, loc=<0 0 0>, rot=<0 0 0 1>}
 base_hair_l_02 {parent=head, loc=<2.0012 4.4971 -7.3871>, rot=<0.0163 0.4306 0.9022 0.0162>}
 dyn_hair_l_02 {parent=base_hair_l_02, loc=<0 0 0>, rot=<0 0 0 1>}
 base_hair_l_03 {parent=head, loc=<2.0789 3.6116 -7.8559>, rot=<-0.1827 0.4969 0.8467 0.0532>}
 dyn_hair_l_03 {parent=base_hair_l_03, loc=<0 0 0>, rot=<0 0 0 1>}
 base_hair_l_04 {parent=head, loc=<2.2019 0.7903 -8.2544>, rot=<-0.0219 0.7061 0.707 0.0334>}
 dyn_hair_l_04 {parent=base_hair_l_04, loc=<0 0 0>, rot=<0 0 0 1>}
 dyn_hair_l_04_front {parent=dyn_hair_l_04, loc=<3.1358 0.4 1.0732>, rot=<0.1388 0.0691 -0.0097 -0.9879>}
 dyn_hair_l_04_back {parent=dyn_hair_l_04, loc=<5.1385 0.1147 -0.9262>, rot=<0.1347 0.0551 0.0802 -0.9861>}
 base_hair_l_05 {parent=head, loc=<1.9944 -2.0533 -8.6342>, rot=<-0.0417 -0.7418 -0.6674 0.0506>}
 dyn_hair_l_05 {parent=base_hair_l_05, loc=<0 0 0>, rot=<0 0 0 1>}
 dyn_hair_l_05_front {parent=dyn_hair_l_05, loc=<1.7688 0.4206 0.6986>, rot=<0 0 0 1>}
 dyn_hair_l_05_mid {parent=dyn_hair_l_05, loc=<1.8535 -0.1564 -0.1177>, rot=<0 0 0 1>}
 dyn_hair_l_05_back {parent=dyn_hair_l_05, loc=<1.914 -0.9926 -1.3546>, rot=<0 0 0 1>}
 base_hair_l_06 {parent=head, loc=<-0.0424 -6.4922 -4.8881>, rot=<-0.0188 -0.9396 -0.3416 0.0129>}
 dyn_hair_l_06 {parent=base_hair_l_06, loc=<0 0 0>, rot=<0 0 0 1>}
 base_hair_r_01 {parent=head, loc=<2.4998 5.4078 6.88>, rot=<-0.0089 -0.5866 0.8088 0.0406>}
 dyn_hair_r_01 {parent=base_hair_r_01, loc=<0 0 0>, rot=<0 0 0 1>}
 base_hair_r_02 {parent=head, loc=<1.7066 4.1869 7.5258>, rot=<0.0315 -0.5561 0.8268 -0.078>}
 dyn_hair_r_02 {parent=base_hair_r_02, loc=<0 0 0>, rot=<0 0 0 1>}
 base_hair_r_03 {parent=head, loc=<2.29 2.5062 8.5428>, rot=<0.0137 -0.6119 0.7878 0.0688>}
 dyn_hair_r_03 {parent=base_hair_r_03, loc=<0 0 0>, rot=<0 0 0 1>}
 base_hair_r_04 {parent=head, loc=<2.2023 0.7903 8.2544>, rot=<0.0213 -0.7183 0.6946 0.0338>}
 dyn_hair_r_04 {parent=base_hair_r_04, loc=<0 0 0>, rot=<0 0 0 1>}
 dyn_hair_r_04_front {parent=base_hair_r_04, loc=<1.9163 -0.6813 -0.5381>, rot=<0 0 0.1994 -0.9799>}
 dyn_hair_r_04_back {parent=base_hair_r_04, loc=<1.9708 -0.4959 0.5803>, rot=<0 0 0.0262 -0.9997>}
 base_hair_r_05 {parent=head, loc=<2.0652 -1.5864 8.0801>, rot=<0.0119 -0.7986 0.6018 -0.0015>}
 dyn_hair_r_05 {parent=base_hair_r_05, loc=<0 0 0>, rot=<0 0 0 1>}
 base_hair_r_06 {parent=head, loc=<-0.0505 -6.0411 5.4262>, rot=<-0.0184 -0.9509 0.3086 -0.0135>}
 dyn_hair_r_06 {parent=base_hair_r_06, loc=<0 0 0>, rot=<0 0 0 1>}
 base_hair_back {parent=head, loc=<0.0692 -8.3672 0.6086>, rot=<-0.0308 -0.9995 0 0>}
 dyn_hair_back {parent=base_hair_back, loc=<0 0 0>, rot=<0 0 0 1>}
 base_glasses {parent=head, loc=<4.191 10.2212 -0.0011>, rot=<-0.0003 -0.0206 0.9996 -0.0216>}
 FACIAL_C_FacialRoot {parent=head, loc=<2.1524 -0.746 -0.0002>, rot=<0.4879 -0.5118 -0.4879 -0.5118>}
 FACIAL_L_CheekInner {parent=FACIAL_C_FacialRoot, loc=<1.9917 0.3169 9.0397>, rot=<-0.1488 -0.0827 -0.1996 -0.965>}
 FACIAL_R_CheekInner {parent=FACIAL_C_FacialRoot, loc=<-1.9926 0.3291 9.0444>, rot=<-0.1488 0.0828 0.1992 -0.9651>}
 FACIAL_C_Nose {parent=FACIAL_C_FacialRoot, loc=<0.0017 -0.3946 10.2798>, rot=<-0.1565 0 -0.0002 -0.9877>}
 FACIAL_C_NoseLower {parent=FACIAL_C_Nose, loc=<0.0081 2.7362 -0.1161>, rot=<0.1305 0 0 -0.9914>}
 FACIAL_L_Nostril {parent=FACIAL_C_Nose, loc=<1.1866 2.4331 -0.2834>, rot=<0.0173 0.1305 0.0023 -0.9913>}
 FACIAL_R_Nostril {parent=FACIAL_C_Nose, loc=<-1.1942 2.4247 -0.2864>, rot=<0.0173 -0.1305 -0.0023 -0.9913>}
 FACIAL_C_NoseTip {parent=FACIAL_C_Nose, loc=<0.0128 2.0157 0.62>, rot=<0.0872 0 0 -0.9962>}
 FACIAL_L_Masseter {parent=FACIAL_C_FacialRoot, loc=<5.7761 1.5937 4.8588>, rot=<0.0431 0.1127 0.0107 -0.9926>}
 FACIAL_R_Masseter {parent=FACIAL_C_FacialRoot, loc=<-5.8364 1.6014 4.8666>, rot=<0.0431 -0.1126 -0.0111 -0.9926>}
 FACIAL_L_Ear {parent=FACIAL_C_FacialRoot, loc=<6.358 -0.0594 1.9741>, rot=<-0.0527 0 -0.0264 -0.9983>}
 FACIAL_R_Ear {parent=FACIAL_C_FacialRoot, loc=<-6.3665 -0.0586 1.9789>, rot=<-0.0527 0 0.026 -0.9983>}
 base_neck_front_l {parent=neck_01, loc=<-1.2243 6.4678 -7.803>, rot=<-0.6931 -0.36 0.0564 -0.6219>}
 dyn_neck_front_l {parent=base_neck_front_l, loc=<0 0 0>, rot=<0 0 0 1>}
 dyn_neck_front_l_end {parent=dyn_neck_front_l, loc=<9.3722 0 0>, rot=<0 0 0 1>}
 base_neck_mid_l {parent=neck_01, loc=<-0.3811 -1.9427 -7.8803>, rot=<-0.496 -0.0357 0.2309 -0.8363>}
 dyn_neck_mid_l {parent=base_neck_mid_l, loc=<0 0 0>, rot=<0 0 0 1>}
 dyn_neck_mid_l_end {parent=dyn_neck_mid_l, loc=<7.2321 0 0>, rot=<0 0 0 1>}
 base_neck_back {parent=neck_01, loc=<-1.1131 -5.2474 0>, rot=<0 0 0.2791 -0.9603>}
 dyn_neck_back {parent=base_neck_back, loc=<0 0 0>, rot=<0 0 0 1>}
 base_neck_mid_r {parent=neck_01, loc=<-0.3811 -1.9426 7.8803>, rot=<0.496 0.0357 0.2309 -0.8363>}
 dyn_neck_mid_r {parent=base_neck_mid_r, loc=<0 0 0>, rot=<0 0 0 -1>}
 dyn_neck_mid_r_end {parent=dyn_neck_mid_r, loc=<7.2326 0.0001 -0.0001>, rot=<0 0 0 1>}
 base_neck_front_r {parent=neck_01, loc=<-1.1213 6.816 7.6801>, rot=<0.8895 0.1639 -0.1855 -0.3841>}
 dyn_neck_front_r {parent=base_neck_front_r, loc=<0 0 0>, rot=<0 0 0 1>}
 dyn_neck_front_r_end {parent=dyn_neck_front_r, loc=<8.0276 0 0>, rot=<0 0 0 1>}
 base_neck_midback_r {parent=neck_01, loc=<-0.9577 -4.3396 4.5083>, rot=<0.2532 -0.0275 0.1432 -0.9564>}
 dyn_neck_midback_r {parent=base_neck_midback_r, loc=<0 0 0>, rot=<0 0 0 1>}
 clavicle_l {parent=spine_05, loc=<4.3575 -0.1466 -2.3223>, rot=<0.0465 -0.7289 -0.009 -0.6829>}
 upperarm_l {parent=clavicle_l, loc=<15.5111 0 0>, rot=<0.0226 -0.3234 -0.0088 -0.9459>}
 upperarm_twist_01_l {parent=upperarm_l, loc=<0.0041 0 0>, rot=<0 0 0 -1>}
 clavicle_r {parent=spine_05, loc=<4.3575 -0.1466 2.3223>, rot=<0.7289 0.0465 -0.6829 0.009>}
 upperarm_r {parent=clavicle_r, loc=<-15.5111 0 0>, rot=<0.0226 -0.3234 -0.0088 -0.9459>}
]
variables [
 head_SPR_look_up_alpha {type=double, default=0}
 isFrontend {type=bool, default=False}
 isParachuteOpen {type=bool, default=False}
 isRigidBodyEnabled {type=bool, default=False}
 isSkyDiveFloating {type=bool, default=False}
 look_up_and_left_two_hands_multiplied_alpha {type=double, default=0}
 upperarm_r_SPR_aim_up_two_hands_alpha {type=double, default=0}
]
