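# Physics asset for the AnvilAlarmMoth male character's face accessory mesh, covering secondary motion of hair and bandana.
# Capsule base bodies attach to the head with free angular constraints, each carrying a dynamic body with locked twist and narrow swing limits.
# Extra front, mid and back strands branch from left dynamic hair bodies and a right base body.
# Head and neck capsules, plus six larger neck capsules angularly locked to neck_01, complete the rig.
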PhysicsAsset /BRCosmetics/Characters/Player/Male/Medium/Bodies/M_MED_AnvilAlarmMoth/Meshes/Parts/M_MED_AnvilAlarmMoth_FaceAcc_Physics
bodies {
 base_bandana_low [{shapeType=Capsule, length=0.5, radius=0.5, shapeName=base_bandana_low_capsule}]
 base_bandana_top [{shapeType=Capsule, length=0.5, radius=0.5, shapeName=base_bandana_top_capsule}]
 base_hair_back [{shapeType=Capsule, length=0.5, radius=0.5, shapeName=base_hair_back_capsule, rotation=<rot 90 0 0>}]
 base_hair_l_01 [{shapeType=Capsule, length=0.5, radius=0.5, shapeName=base_hair_l_01_capsule, rotation=<rot 90 0 0>}]
 base_hair_l_02 [{shapeType=Capsule, length=0.5, radius=0.5, shapeName=base_hair_l_02_capsule, rotation=<rot 90 0 0>}]
 base_hair_l_03 [{shapeType=Capsule, length=0.5, radius=0.5, shapeName=base_hair_l_03_capsule, rotation=<rot 90 0 0>}]
 base_hair_l_04 [{shapeType=Capsule, length=0.5, radius=0.5, shapeName=base_hair_l_04_capsule, rotation=<rot 90 0 0>}]
 base_hair_l_05 [{shapeType=Capsule, length=0.5, radius=0.5, shapeName=base_hair_l_05_capsule, rotation=<rot 90 0 0>}]
 base_hair_l_06 [{shapeType=Capsule, length=0.5, radius=0.5, shapeName=base_hair_l_06_capsule, rotation=<rot 90 0 0>}]
 base_hair_r_01 [{shapeType=Capsule, length=0.5, radius=0.5, shapeName=base_hair_r_01_capsule, rotation=<rot 90 0 0>}]
 base_hair_r_02 [{shapeType=Capsule, length=0.5, radius=0.5, shapeName=base_hair_r_02_capsule, rotation=<rot 90 0 0>}]
 base_hair_r_03 [{shapeType=Capsule, length=0.5, radius=0.5, shapeName=base_hair_r_03_capsule, rotation=<rot 90 0 0>}]
 base_hair_r_04 [{shapeType=Capsule, length=0.5, radius=0.5, shapeName=base_hair_r_04_capsule, rotation=<rot 90 0 0>}]
 base_hair_r_05 [{shapeType=Capsule, length=0.5, radius=0.5, shapeName=base_hair_r_05_capsule, rotation=<rot 90 0 0>}]
 base_hair_r_06 [{shapeType=Capsule, length=0.5, radius=0.5, shapeName=base_hair_r_06_capsule, rotation=<rot 90 0 0>}]
 dyn_bandana_low_01 [{shapeType=Capsule, length=4.9594, radius=1.24, shapeName=dyn_bandana_low_01_capsule, center=<3.2631 -0.4192 0>, rotation=<rot 90 0 0>}]
 dyn_bandana_top [{shapeType=Capsule, length=1.4361, radius=1.201, shapeName=dyn_bandana_top_capsule, center=<1.9595 -0.4784 0>, rotation=<rot 90 0 0>}]
 dyn_hair_back [{shapeType=Capsule, length=4, radius=2, shapeName=dyn_hair_back_capsule, center=<4.286 -0.9319 0>, rotation=<rot 90 0 0>}]
 dyn_hair_l_01 [{shapeType=Capsule, length=4, radius=0.7, shapeName=dyn_hair_l_01_capsule, center=<2.051 -0.3879 0>, rotation=<rot 90 -9.3349 0>}]
 dyn_hair_l_02 [{shapeType=Capsule, length=5.6692, radius=0.7, shapeName=dyn_hair_l_02_capsule, center=<2.9626 0.384 0>, rotation=<rot 90 7.2433 0>}]
 dyn_hair_l_03 [{shapeType=Capsule, length=4, radius=0.7, shapeName=dyn_hair_l_03_capsule, center=<2.3773 -0.0129 0>, rotation=<rot 90 -3.8426 0>}]
 dyn_hair_l_04 [{shapeType=Capsule, length=3.3338, radius=1.3806, shapeName=dyn_hair_l_04_capsule, center=<2.6857 -0.2978 0>, rotation=<rot 90 7.1705 0>}]
 dyn_hair_l_04_back [{
   shapeType=Capsule
   length=3.2953
   radius=1.1184
   shapeName=dyn_hair_l_04_back_capsule
   center=<2.0127 -0.0289 -0.0213>
   rotation=<rot 0.0979 -96.3611 -87.3798>
  }]
 dyn_hair_l_04_front [{
   shapeType=Capsule
   length=3.8358
   radius=0.5778
   shapeName=dyn_hair_l_04_back_capsule
   center=<1.4649 -0.2626 0>
   rotation=<rot 90 -15.411 0>
  }]
 dyn_hair_l_05 [{
   shapeType=Capsule
   length=3.8316
   radius=1.1901
   shapeName=dyn_hair_l_05_capsule
   center=<3.1058 -0.1578 -0.0209>
   rotation=<rot 90 -3.4481 0>
  }]
 dyn_hair_l_05_back [{
   shapeType=Capsule
   length=2.3255
   radius=0.5901
   shapeName=dyn_hair_l_05_back_capsule
   center=<1.0958 -0.1772 0.1255>
   rotation=<rot -0.1344 -96.8224 -88.6754>
  }]
 dyn_hair_l_05_front [{
   shapeType=Capsule
   length=4.2632
   radius=0.6223
   shapeName=dyn_hair_l_05_front_capsule
   center=<3.0024 0.1652 0.2766>
   rotation=<rot 0.3251 84.3077 -85.1904>
  }]
 dyn_hair_l_05_mid [{
   shapeType=Capsule
   length=5.5917
   radius=0.7214
   shapeName=dyn_hair_l_05_mid_capsule
   center=<3.5158 0.0934 -0.2023>
   rotation=<rot 0 -86.2164 -87.664>
  }]
 dyn_hair_l_06 [{shapeType=Capsule, length=5.7512, radius=1.6764, shapeName=dyn_hair_l_06_capsule, center=<3.1201 0.4785 0.636>, rotation=<rot 0 0 -2.1002>}]
 dyn_hair_r_01 [{shapeType=Capsule, length=5.3432, radius=0.7, shapeName=dyn_hair_r_01_capsule, center=<2.5813 0.0073 0>, rotation=<rot 90 3.5013 0>}]
 dyn_hair_r_02 [{
   shapeType=Capsule
   length=5.4977
   radius=1.0376
   shapeName=dyn_hair_r_02_capsule
   center=<2.9026 -0.6664 -0.0781>
   rotation=<rot 84.4887 -0.0056 15.6118>
  }]
 dyn_hair_r_03 [{
   shapeType=Capsule
   length=4
   radius=0.7
   shapeName=dyn_hair_r_03_capsule
   center=<1.6694 0.0016 -0.1147>
   rotation=<rot 82.5316 -6.2665 5.252>
  }]
 dyn_hair_r_04 [{shapeType=Capsule, length=2.7749, radius=1.1359, shapeName=dyn_hair_r_04_capsule, center=<2.2159 -0.3392 0.1365>, rotation=<rot 90 2.1336 0>}]
 dyn_hair_r_04_back [{
   shapeType=Capsule
   length=3.433
   radius=0.6587
   shapeName=dyn_hair_r_04_back_capsule
   center=<2.336 -0.1414 0.4039>
   rotation=<rot 0 95.1693 -79.9151>
  }]
 dyn_hair_r_04_front [{
   shapeType=Capsule
   length=3.691
   radius=0.7973
   shapeName=dyn_hair_r_04_front_capsule
   center=<1.6879 0.0395 0.0636>
   rotation=<rot 0 87.095 -84.2219>
  }]
 dyn_hair_r_05 [{shapeType=Capsule, length=6.3772, radius=1.3536, shapeName=dyn_hair_r_05_capsule, center=<3.2531 -0.3277 0>, rotation=<rot 90 0.214 0>}]
 dyn_hair_r_06 [{shapeType=Capsule, length=5.7512, radius=1.6764, shapeName=dyn_hair_l_06_capsule, center=<3.1201 0.4785 0.636>, rotation=<rot 0 0 -2.1002>}]
 dyn_neck_back [{shapeType=Capsule, length=6, radius=4, shapeName=dyn_neck_front_l_capsule, center=<6 -3 0>}]
 dyn_neck_front_l [{shapeType=Capsule, length=6, radius=4, shapeName=dyn_neck_front_l_capsule, center=<5 -2 2.5>, rotation=<rot 25.7197 0 -1.7262>}]
 dyn_neck_front_r [{shapeType=Capsule, length=6, radius=4, shapeName=dyn_neck_front_l_capsule, center=<4 -2 -3.5>, rotation=<rot -26.2281 -3.3852 -4.9253>}]
 dyn_neck_mid_l [{shapeType=Capsule, length=6, radius=4, shapeName=dyn_neck_front_l_capsule, center=<5 -2 2.5>, rotation=<rot -1.4313 -0.3629 14.2221>}]
 dyn_neck_mid_r [{shapeType=Capsule, length=6, radius=4, shapeName=dyn_neck_front_l_capsule, center=<5 -2 2>, rotation=<rot 13.7701 3.1222 11.0827>}]
 dyn_neck_midback_r [{shapeType=Capsule, length=6, radius=4, shapeName=dyn_neck_front_l_capsule, center=<4.9356 -4.5 -0.0852>, rotation=<rot 4.5788 0 0>}]
 head [{shapeType=Capsule, length=10.2002, radius=7.2998, center=<0 0.4835 0>, rotation=<rot 90 0 0>}]
 neck_01 [{shapeType=Capsule, length=9.3575, radius=5.1497, shapeName=neck_01_capsule, rotation=<rot 0 -90.0002 -88.6601>}]
}
constraints [
 {
  bone1Name=base_hair_back
  bone2Name=head
  swing1LimitDegrees=45
  swing1Motion=Free
  swing2LimitDegrees=45
  swing2Motion=Free
  twistLimitDegrees=45
  twistMotion=Free
 }
 {bone1Name=dyn_hair_back, bone2Name=base_hair_back, swing1LimitDegrees=0, swing2LimitDegrees=10, twistLimitDegrees=45, twistMotion=Locked}
 {
  bone1Name=base_hair_l_01
  bone2Name=head
  swing1LimitDegrees=45
  swing1Motion=Free
  swing2LimitDegrees=45
  swing2Motion=Free
  twistLimitDegrees=45
  twistMotion=Free
 }
 {bone1Name=dyn_hair_l_01, bone2Name=base_hair_l_01, swing1LimitDegrees=10, swing2LimitDegrees=10, twistLimitDegrees=45, twistMotion=Locked}
 {
  bone1Name=base_hair_l_02
  bone2Name=head
  swing1LimitDegrees=45
  swing1Motion=Free
  swing2LimitDegrees=45
  swing2Motion=Free
  twistLimitDegrees=45
  twistMotion=Free
 }
 {bone1Name=dyn_hair_l_02, bone2Name=base_hair_l_02, swing1LimitDegrees=7, swing2LimitDegrees=7, twistLimitDegrees=45, twistMotion=Locked}
 {
  bone1Name=base_hair_l_03
  bone2Name=head
  swing1LimitDegrees=45
  swing1Motion=Free
  swing2LimitDegrees=45
  swing2Motion=Free
  twistLimitDegrees=45
  twistMotion=Free
 }
 {bone1Name=dyn_hair_l_03, bone2Name=base_hair_l_03, swing1LimitDegrees=5, swing2LimitDegrees=5, twistLimitDegrees=45, twistMotion=Locked}
 {
  bone1Name=base_hair_l_04
  bone2Name=head
  swing1LimitDegrees=45
  swing1Motion=Free
  swing2LimitDegrees=45
  swing2Motion=Free
  twistLimitDegrees=45
  twistMotion=Free
 }
 {bone1Name=dyn_hair_l_04, bone2Name=base_hair_l_04, swing1LimitDegrees=5, swing2LimitDegrees=5, twistLimitDegrees=45, twistMotion=Locked}
 {
  bone1Name=base_hair_l_05
  bone2Name=head
  swing1LimitDegrees=45
  swing1Motion=Free
  swing2LimitDegrees=45
  swing2Motion=Free
  twistLimitDegrees=45
  twistMotion=Free
 }
 {bone1Name=dyn_hair_l_05, bone2Name=base_hair_l_05, swing1LimitDegrees=5, swing2LimitDegrees=5, twistLimitDegrees=45, twistMotion=Locked}
 {
  bone1Name=base_hair_l_06
  bone2Name=head
  swing1LimitDegrees=45
  swing1Motion=Free
  swing2LimitDegrees=45
  swing2Motion=Free
  twistLimitDegrees=45
  twistMotion=Free
 }
 {bone1Name=dyn_hair_l_06, bone2Name=base_hair_l_06, swing1LimitDegrees=0, swing2LimitDegrees=10, twistLimitDegrees=45, twistMotion=Locked}
 {
  bone1Name=base_hair_r_01
  bone2Name=head
  swing1LimitDegrees=45
  swing1Motion=Free
  swing2LimitDegrees=45
  swing2Motion=Free
  twistLimitDegrees=45
  twistMotion=Free
 }
 {bone1Name=dyn_hair_r_01, bone2Name=base_hair_r_01, swing1LimitDegrees=10, swing2LimitDegrees=10, twistLimitDegrees=45, twistMotion=Locked}
 {
  bone1Name=base_hair_r_02
  bone2Name=head
  swing1LimitDegrees=45
  swing1Motion=Free
  swing2LimitDegrees=45
  swing2Motion=Free
  twistLimitDegrees=45
  twistMotion=Free
 }
 {bone1Name=dyn_hair_r_02, bone2Name=base_hair_r_02, swing1LimitDegrees=8, swing2LimitDegrees=8, twistLimitDegrees=45, twistMotion=Locked}
 {
  bone1Name=base_hair_r_03
  bone2Name=head
  swing1LimitDegrees=45
  swing1Motion=Free
  swing2LimitDegrees=45
  swing2Motion=Free
  twistLimitDegrees=45
  twistMotion=Free
 }
 {bone1Name=dyn_hair_r_03, bone2Name=base_hair_r_03, swing1LimitDegrees=7, swing2LimitDegrees=7, twistLimitDegrees=45, twistMotion=Locked}
 {
  bone1Name=base_hair_r_04
  bone2Name=head
  swing1LimitDegrees=45
  swing1Motion=Free
  swing2LimitDegrees=45
  swing2Motion=Free
  twistLimitDegrees=45
  twistMotion=Free
 }
 {bone1Name=dyn_hair_r_04, bone2Name=base_hair_r_04, swing1LimitDegrees=7, swing2LimitDegrees=6, twistLimitDegrees=45, twistMotion=Locked}
 {
  bone1Name=base_hair_r_05
  bone2Name=head
  swing1LimitDegrees=45
  swing1Motion=Free
  swing2LimitDegrees=45
  swing2Motion=Free
  twistLimitDegrees=45
  twistMotion=Free
 }
 {bone1Name=dyn_hair_r_05, bone2Name=base_hair_r_05, swing1LimitDegrees=10, swing2LimitDegrees=10, twistLimitDegrees=45, twistMotion=Locked}
 {
  bone1Name=base_hair_r_06
  bone2Name=head
  swing1LimitDegrees=45
  swing1Motion=Free
  swing2LimitDegrees=45
  swing2Motion=Free
  twistLimitDegrees=45
  twistMotion=Free
 }
 {bone1Name=dyn_hair_r_06, bone2Name=base_hair_r_06, swing1LimitDegrees=0, swing2LimitDegrees=10, twistLimitDegrees=45, twistMotion=Locked}
 {
  bone1Name=base_bandana_low
  bone2Name=head
  swing1LimitDegrees=45
  swing1Motion=Free
  swing2LimitDegrees=45
  swing2Motion=Free
  twistLimitDegrees=45
  twistMotion=Free
 }
 {
  bone1Name=dyn_bandana_low_01
  bone2Name=base_bandana_low
  swing1LimitDegrees=20
  swing2LimitDegrees=25
  twistLimitDegrees=45
  twistMotion=Locked
 }
 {
  bone1Name=base_bandana_top
  bone2Name=head
  swing1LimitDegrees=45
  swing1Motion=Free
  swing2LimitDegrees=45
  swing2Motion=Free
  twistLimitDegrees=45
  twistMotion=Free
 }
 {
  bone1Name=dyn_bandana_top
  bone2Name=base_bandana_top
  swing1LimitDegrees=10
  swing2LimitDegrees=20
  twistLimitDegrees=45
  twistMotion=Locked
 }
 {bone1Name=dyn_hair_l_04_back, bone2Name=dyn_hair_l_04, swing1LimitDegrees=10, swing2LimitDegrees=7, twistLimitDegrees=45, twistMotion=Locked}
 {
  bone1Name=dyn_hair_l_04_front
  bone2Name=dyn_hair_l_04
  swing1LimitDegrees=10
  swing2LimitDegrees=10
  twistLimitDegrees=45
  twistMotion=Locked
 }
 {bone1Name=dyn_hair_l_05_back, bone2Name=dyn_hair_l_05, swing1LimitDegrees=6, swing2LimitDegrees=6, twistLimitDegrees=45, twistMotion=Locked}
 {bone1Name=dyn_hair_l_05_front, bone2Name=dyn_hair_l_05, swing1LimitDegrees=6, swing2LimitDegrees=6, twistLimitDegrees=45, twistMotion=Locked}
 {bone1Name=dyn_hair_l_05_mid, bone2Name=dyn_hair_l_05, swing1LimitDegrees=6, swing2LimitDegrees=6, twistLimitDegrees=45, twistMotion=Locked}
 {bone1Name=dyn_hair_r_04_back, bone2Name=base_hair_r_04, swing1LimitDegrees=5, swing2LimitDegrees=5, twistLimitDegrees=45, twistMotion=Locked}
 {
  bone1Name=dyn_hair_r_04_front
  bone2Name=base_hair_r_04
  swing1LimitDegrees=10
  swing2LimitDegrees=5
  twistLimitDegrees=45
  twistMotion=Locked
 }
 {
  bone1Name=head
  bone2Name=neck_01
  swing1LimitDegrees=45
  swing1Motion=Free
  swing2LimitDegrees=45
  swing2Motion=Free
  twistLimitDegrees=45
  twistMotion=Free
 }
 {
  bone1Name=dyn_neck_front_l
  bone2Name=neck_01
  swing1LimitDegrees=45
  swing1Motion=Locked
  swing2LimitDegrees=45
  swing2Motion=Locked
  twistLimitDegrees=45
  twistMotion=Locked
 }
 {
  bone1Name=dyn_neck_mid_l
  bone2Name=neck_01
  swing1LimitDegrees=45
  swing1Motion=Locked
  swing2LimitDegrees=45
  swing2Motion=Locked
  twistLimitDegrees=45
  twistMotion=Locked
 }
 {
  bone1Name=dyn_neck_back
  bone2Name=neck_01
  swing1LimitDegrees=45
  swing1Motion=Locked
  swing2LimitDegrees=45
  swing2Motion=Locked
  twistLimitDegrees=45
  twistMotion=Locked
 }
 {
  bone1Name=dyn_neck_mid_r
  bone2Name=neck_01
  swing1LimitDegrees=45
  swing1Motion=Locked
  swing2LimitDegrees=45
  swing2Motion=Locked
  twistLimitDegrees=45
  twistMotion=Locked
 }
 {
  bone1Name=dyn_neck_front_r
  bone2Name=neck_01
  swing1LimitDegrees=45
  swing1Motion=Locked
  swing2LimitDegrees=45
  swing2Motion=Locked
  twistLimitDegrees=45
  twistMotion=Locked
 }
 {
  bone1Name=dyn_neck_midback_r
  bone2Name=neck_01
  swing1LimitDegrees=45
  swing1Motion=Locked
  swing2LimitDegrees=45
  swing2Motion=Locked
  twistLimitDegrees=45
  twistMotion=Locked
 }
]
